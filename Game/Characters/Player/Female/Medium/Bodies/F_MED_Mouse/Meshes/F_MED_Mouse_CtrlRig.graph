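# Control Rig for the F_MED_Mouse player body that applies corrective glute deformation driven by thigh pose.
# A SphericalPoseReader on each thigh bone measures its orientation, with mirrored axes, rotation offsets and width scaling for left and right.
# Each reader's output weights a ModifyTransforms node that offsets and enlarges the matching dyn_L_gluteFix or dyn_R_gluteFix bone.
# The rest of the imported skeleton, including fur, tie and bracelet bones, is left untouched.
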ControlRigBlueprint /Game/Characters/Player/Female/Medium/Bodies/F_MED_Mouse/Meshes/F_MED_Mouse_CtrlRig
graphs [
 RigVMFunctionLibrary {links [], nodes []}
 RigVMModel {nodes [
   RigUnit_BeginExecution = &RigUnit_BeginExecution()
   n1 = &RigUnit_SphericalPoseReader(
    ActiveRegionScaleFactors=(PositiveWidth=3.5,NegativeWidth=0.75,PositiveHeight=3,NegativeHeight=3)
    ActiveRegionSize=0.1
    Debug=(bDrawDebug=false,DebugScale=25,DebugSegments=20,DebugThickness=0.25)
    DriverAxis=(X=-1,Y=0,Z=0)
    DriverItem=(Type=Bone,Name="thigh_l")
    FalloffRegionScaleFactors=(PositiveWidth=1,NegativeWidth=1,PositiveHeight=1,NegativeHeight=1)
    FalloffSize=0.45
    OptionalParentItem=(Type=Bone)
    RotationOffset=(X=90,Y=0,Z=0)
    ExecutePin=@RigUnit_BeginExecution.ExecutePin
   )
   > n2 = &RigUnit_SphericalPoseReader(
    ActiveRegionScaleFactors=(PositiveWidth=3.5,NegativeWidth=0.75,PositiveHeight=3,NegativeHeight=3)
    ActiveRegionSize=0.1
    Debug=(bDrawDebug=false,DebugScale=25,DebugSegments=20,DebugThickness=0.25)
    DriverAxis=(X=1,Y=0,Z=0)
    DriverItem=(Type=Bone,Name="thigh_r")
    FalloffRegionScaleFactors=(PositiveWidth=1,NegativeWidth=1,PositiveHeight=1,NegativeHeight=1)
    FalloffSize=0.45
    FlipHeightScaling=false
    FlipWidthScaling=true
    OptionalParentItem=(Type=Bone)
    RotationOffset=(X=-90,Y=0,Z=0)
   )
   > &RigUnit_ModifyTransforms(
    ItemToModify=((Item=(Type=Bone,Name="dyn_L_gluteFix"),Transform=(Rotation=(X=0,Y=0,Z=0,W=1),Translation=(X=2,Y=0,Z=0),Scale3D=(X=1.1,Y=1.3,Z=1.3))))
    Weight=@n1.OutputParam
   )
   > &RigUnit_ModifyTransforms(
    ItemToModify=((Item=(Type=Bone,Name="dyn_R_gluteFix"),Transform=(Rotation=(X=0,Y=0,Z=0,W=1),Translation=(X=-2,Y=0,Z=0),Scale3D=(X=1.1,Y=1.3,Z=1.3))))
    Weight=@n2.OutputParam
   )
  ]}
]
hierarchy [
 root {loc=<0 0 0>, rot=<0 0 0 1>}
 pelvis {parent=root, loc=<0 -0.5301 98.0571>, rot=<0.0127 0.707 -0.0127 -0.707>}
 spine_01 {parent=pelvis, loc=<6.1208 -0.9328 0>, rot=<0 0 0.034 -0.9994>}
 spine_02 {parent=spine_01, loc=<6.5996 -0.2095 0>, rot=<0 0 0.0238 -0.9997>}
 spine_03 {parent=spine_02, loc=<6.4914 -0.1237 0>, rot=<0 0 0.0149 -0.9999>}
 spine_04 {parent=spine_03, loc=<6.9598 0.0794 0>, rot=<0 0 -0.0132 -0.9999>}
 spine_05 {parent=spine_04, loc=<6.9987 0 0>, rot=<0 0 -0.0015 -1>}
 clavicle_l {parent=spine_05, loc=<5.4831 -0.1053 -3.2946>, rot=<0.0465 -0.7289 -0.009 -0.6829>}
 upperarm_l {parent=clavicle_l, loc=<9.8658 -1.2136 -0.4105>, rot=<0.0164 -0.3245 -0.0269 -0.9454>}
 lowerarm_l {parent=upperarm_l, loc=<26.465 -0.8345 0.259>, rot=<0.0036 0.0161 -0.2298 -0.9731>}
 hand_l {parent=lowerarm_l, loc=<22.7355 -0.3203 -0.6105>, rot=<-0.6784 -0.0223 0.0735 -0.7307>}
 thumb_01_l {parent=hand_l, loc=<1.3118 -0.3191 -2.459>, rot=<0.6546 -0.4042 0.0664 -0.6354>}
 thumb_02_l {parent=thumb_01_l, loc=<5.9336 -0.0448 0.0401>, rot=<-0.0037 0.0213 0.298 -0.9543>}
 thumb_03_l {parent=thumb_02_l, loc=<2.2223 0.3682 -0.0907>, rot=<0.0118 -0.013 0.0066 -0.9998>}
 index_metacarpal_l {parent=hand_l, loc=<1.6535 0.0183 -2.96>, rot=<0.0158 -0.0124 -0.0173 -0.9996>}
 index_01_l {parent=index_metacarpal_l, loc=<8.3008 0.0357 -0.0524>, rot=<-0.0044 0.0485 0.0924 -0.9945>}
 index_02_l {parent=index_01_l, loc=<3.3587 -0.3004 0.0452>, rot=<-0.002 0.0027 0.3225 -0.9466>}
 index_03_l {parent=index_02_l, loc=<2.5578 -0.123 -0.0252>, rot=<-0.002 0.0067 0.2558 -0.9667>}
 middle_metacarpal_l {parent=hand_l, loc=<2.1856 -0.1725 -0.5984>, rot=<0.006 0.0168 -0.0329 -0.9993>}
 middle_01_l {parent=middle_metacarpal_l, loc=<7.3275 -0.2447 0.0129>, rot=<-0.0208 0.0042 0.0858 -0.9961>}
 middle_02_l {parent=middle_01_l, loc=<4.0123 0.1401 0.1175>, rot=<-0.0218 0.0258 0.3512 -0.9357>}
 middle_03_l {parent=middle_02_l, loc=<2.6077 0.0534 0.1373>, rot=<0.0082 0.0188 0.1708 -0.9851>}
 ring_metacarpal_l {parent=hand_l, loc=<2.2074 -0.3625 1.3638>, rot=<-0.0284 0.0722 -0.0458 -0.9959>}
 ring_01_l {parent=ring_metacarpal_l, loc=<7.0295 -0.6197 -0.1446>, rot=<-0.0489 -0.0252 0.1892 -0.9804>}
 ring_02_l {parent=ring_01_l, loc=<3.6174 0.4647 -0.1027>, rot=<0 0 0.2806 -0.9598>}
 ring_03_l {parent=ring_02_l, loc=<2.4152 0.114 0.0199>, rot=<0 0 0.1189 -0.9929>}
 pinky_metacarpal_l {parent=hand_l, loc=<1.7616 -0.9753 3.3814>, rot=<-0.0397 0.0814 -0.069 -0.9935>}
 pinky_01_l {parent=pinky_metacarpal_l, loc=<7.1709 -0.4831 -0.1094>, rot=<-0.0654 -0.0035 0.2242 -0.9723>}
 pinky_02_l {parent=pinky_01_l, loc=<2.8274 0.0893 -0.1043>, rot=<0.0037 0.0011 0.2854 -0.9584>}
 pinky_03_l {parent=pinky_02_l, loc=<2.2046 0.0307 -0.0078>, rot=<-0.0002 0.0098 0.1445 -0.9895>}
 lowerarm_twist_02_l {parent=lowerarm_l, loc=<10.6707 0 0>, rot=<0 0 0 1>}
 dyn_L_arm_fur_4 {parent=lowerarm_twist_02_l, loc=<-0.7265 1.0243 1.8031>, rot=<0.3077 0.6282 -0.0951 -0.7083>}
 dyn_L_arm_fur_5 {parent=lowerarm_twist_02_l, loc=<-1.0527 -2.5554 -0.5789>, rot=<-0.5562 0.3749 0.497 -0.5505>}
 lowerarm_twist_01_l {parent=lowerarm_l, loc=<18.316 0 0>, rot=<0 0 0 1>}
 dyn_L_arm_fur_3 {parent=lowerarm_l, loc=<5.4616 1.0959 -1.5946>, rot=<0.6823 0.1854 -0.5259 -0.4728>}
 dyn_L_bracelet_1 {parent=lowerarm_l, loc=<19.4497 0.0423 -0.1717>, rot=<0 0 0 -1>}
 dyn_L_bracelet_2 {parent=lowerarm_l, loc=<20.6045 -0.2519 -0.1591>, rot=<0 0 0 -1>}
 dyn_L_bracelet_charm_A {parent=dyn_L_bracelet_2, loc=<0.2715 -2.7343 2.0071>, rot=<0.9312 -0.1572 0.1337 -0.3004>}
 dyn_L_bracelet_charm_B {parent=dyn_L_bracelet_2, loc=<0.5816 1.032 2.4079>, rot=<0.9269 0.0141 0.3643 0.0895>}
 dyn_L_bracelet_charm_C {parent=dyn_L_bracelet_2, loc=<-0.0302 3.4005 -1.4038>, rot=<-0.5099 -0.4742 -0.1091 -0.7094>}
 dyn_L_bracelet_charm_D {parent=dyn_L_bracelet_2, loc=<-0.5999 -1.7564 -2.4081>, rot=<0.8342 -0.0735 -0.5419 0.0707>}
 dyn_L_bracelet_3 {parent=lowerarm_l, loc=<22.4655 0.0484 -0.5098>, rot=<0 0 0 -1>}
 upperarm_twist_01_l {parent=upperarm_l, loc=<7.2135 -0.1134 0>, rot=<0 0 0 1>}
 dyn_L_arm_fur_2 {parent=upperarm_twist_01_l, loc=<-0.858 -0.4484 0.3569>, rot=<-0.0039 0.7572 0.0065 -0.6531>}
 upperarm_twist_02_l {parent=upperarm_l, loc=<19.7998 -0.1134 0>, rot=<0 0 0 1>}
 dyn_L_arm_fur_1 {parent=clavicle_l, loc=<4.4623 0.205 2.6624>, rot=<-0.013 0.5726 -0.0216 -0.8195>}
 neck_01 {parent=spine_05, loc=<11.3138 0.1529 0>, rot=<0 0 -0.2003 -0.9797>}
 neck_02 {parent=neck_01, loc=<4.5627 0 0>, rot=<0 0 0.0504 -0.9987>}
 head {parent=neck_02, loc=<5.3914 0 0>, rot=<0 0 0.115 -0.9934>}
 clavicle_r {parent=spine_05, loc=<5.4831 -0.1053 3.2946>, rot=<0.7289 0.0465 -0.6829 0.009>}
 upperarm_r {parent=clavicle_r, loc=<-9.8658 1.2136 0.4105>, rot=<0.0164 -0.3245 -0.0269 -0.9454>}
 lowerarm_r {parent=upperarm_r, loc=<-26.465 0.8345 -0.259>, rot=<0.0036 0.0161 -0.2298 -0.9731>}
 hand_r {parent=lowerarm_r, loc=<-22.7355 0.3203 0.6106>, rot=<-0.6784 -0.0223 0.0735 -0.7307>}
 thumb_01_r {parent=hand_r, loc=<-1.3118 0.3191 2.459>, rot=<0.6546 -0.4042 0.0664 -0.6354>}
 thumb_02_r {parent=thumb_01_r, loc=<-5.9336 0.0448 -0.0401>, rot=<-0.0037 0.0213 0.298 -0.9543>}
 thumb_03_r {parent=thumb_02_r, loc=<-2.2223 -0.3682 0.0907>, rot=<0.0118 -0.013 0.0066 -0.9998>}
 index_metacarpal_r {parent=hand_r, loc=<-1.6535 -0.0183 2.96>, rot=<0.0158 -0.0124 -0.0173 -0.9996>}
 index_01_r {parent=index_metacarpal_r, loc=<-8.3008 -0.0357 0.0524>, rot=<-0.0044 0.0485 0.0924 -0.9945>}
 index_02_r {parent=index_01_r, loc=<-3.3587 0.3004 -0.0452>, rot=<-0.002 0.0027 0.3225 -0.9466>}
 index_03_r {parent=index_02_r, loc=<-2.5578 0.123 0.0252>, rot=<-0.002 0.0067 0.2558 -0.9667>}
 middle_metacarpal_r {parent=hand_r, loc=<-2.1856 0.1725 0.5984>, rot=<0.006 0.0168 -0.0329 -0.9993>}
 middle_01_r {parent=middle_metacarpal_r, loc=<-7.3275 0.2447 -0.0129>, rot=<-0.0208 0.0042 0.0858 -0.9961>}
 middle_02_r {parent=middle_01_r, loc=<-4.0123 -0.1401 -0.1175>, rot=<-0.0218 0.0258 0.3512 -0.9357>}
 middle_03_r {parent=middle_02_r, loc=<-2.6077 -0.0534 -0.1373>, rot=<0.0082 0.0188 0.1708 -0.9851>}
 ring_metacarpal_r {parent=hand_r, loc=<-2.2074 0.3625 -1.3638>, rot=<-0.0284 0.0722 -0.0458 -0.9959>}
 ring_01_r {parent=ring_metacarpal_r, loc=<-7.0295 0.6197 0.1446>, rot=<-0.0489 -0.0252 0.1892 -0.9804>}
 ring_02_r {parent=ring_01_r, loc=<-3.6174 -0.4647 0.1027>, rot=<0 0 0.2806 -0.9598>}
 ring_03_r {parent=ring_02_r, loc=<-2.4152 -0.114 -0.0199>, rot=<0 0 0.1189 -0.9929>}
 pinky_metacarpal_r {parent=hand_r, loc=<-1.7616 0.9753 -3.3814>, rot=<-0.0397 0.0814 -0.069 -0.9935>}
 pinky_01_r {parent=pinky_metacarpal_r, loc=<-7.1709 0.4831 0.1094>, rot=<-0.0654 -0.0035 0.2242 -0.9723>}
 pinky_02_r {parent=pinky_01_r, loc=<-2.8274 -0.0893 0.1043>, rot=<0.0037 0.0011 0.2854 -0.9584>}
 pinky_03_r {parent=pinky_02_r, loc=<-2.2046 -0.0307 0.0078>, rot=<-0.0002 0.0098 0.1445 -0.9895>}
 lowerarm_twist_02_r {parent=lowerarm_r, loc=<-10.6707 0 0>, rot=<0 0 0 -1>}
 dyn_R_arm_fur_5 {parent=lowerarm_twist_02_r, loc=<0.7264 -1.0243 -1.803>, rot=<0.0375 -0.8252 0.1427 -0.5452>}
 dyn_R_arm_fur_6 {parent=lowerarm_twist_02_r, loc=<0.5826 2.735 0.4055>, rot=<0.5471 0.4575 0.5833 0.3887>}
 dyn_R_arm_fur_7 {parent=lowerarm_twist_02_r, loc=<0.6423 -1.8136 1.9123>, rot=<0.6179 -0.7414 0.1753 0.1944>}
 lowerarm_twist_01_r {parent=lowerarm_r, loc=<-18.316 0 0>, rot=<0 0 0 -1>}
 dyn_R_arm_fur_4 {parent=lowerarm_r, loc=<-5.4748 -0.3361 -0.5991>, rot=<0.0241 -0.7514 0.0907 -0.6531>}
 dyn_R_bracelet_1 {parent=lowerarm_r, loc=<-19.4494 -0.0424 0.1714>, rot=<0 0 0 1>}
 dyn_R_bracelet_2 {parent=lowerarm_r, loc=<-20.6046 0.2519 0.1592>, rot=<0 0 0 1>}
 dyn_R_bracelet_charm_A {parent=dyn_R_bracelet_2, loc=<-0.2717 2.7344 -2.0069>, rot=<-0.1337 -0.3004 0.9312 0.1572>}
 dyn_R_bracelet_charm_B {parent=dyn_R_bracelet_2, loc=<-0.5814 -1.032 -2.4082>, rot=<-0.3643 0.0895 0.9269 -0.0141>}
 dyn_R_bracelet_charm_C {parent=dyn_R_bracelet_2, loc=<0.0304 -3.4006 1.4036>, rot=<0.1091 -0.7094 -0.5099 0.4742>}
 dyn_R_bracelet_charm_D {parent=dyn_R_bracelet_2, loc=<0.6 1.7563 2.4079>, rot=<0.5419 0.0707 0.8342 0.0735>}
 dyn_R_bracelet_3 {parent=lowerarm_r, loc=<-22.4656 -0.0484 0.51>, rot=<0 0 0 1>}
 upperarm_twist_01_r {parent=upperarm_r, loc=<-7.2135 0.1134 0>, rot=<0 0 0 -1>}
 upperarm_twist_02_r {parent=upperarm_r, loc=<-19.7998 0.1134 0>, rot=<0 0 0 -1>}
 dyn_R_arm_fur_3 {parent=upperarm_r, loc=<0.9988 0.9651 -0.5126>, rot=<-0.1849 -0.7312 -0.1607 -0.6366>}
 dyn_R_arm_fur_1 {parent=clavicle_r, loc=<-6.15 -0.2903 -2.7759>, rot=<0.0222 -0.8481 -0.0119 -0.5292>}
 dyn_R_arm_fur_2 {parent=clavicle_r, loc=<-8.9301 -3.1741 -0.5856>, rot=<0.3008 -0.8764 0.1053 -0.361>}
 dyn_shoulderInflate {parent=clavicle_r, loc=<-8.6679 -0.3354 -2.5994>, rot=<0.0104 -0.7557 0.0385 -0.6537>}
 attach_backpack {parent=spine_05, loc=<-0.1243 -6.8045 0>, rot=<-0.0093 -0.707 0.0093 -0.707>}
 attach_cape {parent=spine_05, loc=<6.2596 -6.2409 0>, rot=<0.0616 -0.7044 -0.0616 -0.7044>}
 dyn_tieBase {parent=spine_05, loc=<5.4569 7.3258 0.0702>, rot=<-0.1693 -0.6865 -0.6742 -0.2133>}
 dyn_underTie_1 {parent=dyn_tieBase, loc=<7.874 0 0>, rot=<0.0061 -0.0753 0.0158 -0.997>}
 dyn_underTie_2 {parent=dyn_underTie_1, loc=<4.5105 0 0>, rot=<0.0008 -0.0077 0.0048 -1>}
 dyn_underTie_3 {parent=dyn_underTie_2, loc=<4.7266 0 0>, rot=<0.0002 -0.0015 0.0027 -1>}
 dyn_underTie_4 {parent=dyn_underTie_3, loc=<5.4489 0 0>, rot=<0 0 0 1>}
 dyn_overTie_1 {parent=dyn_tieBase, loc=<9.2376 0.7259 0.4871>, rot=<-0.0003 -0.078 -0.0682 -0.9946>}
 dyn_overTie_2 {parent=dyn_overTie_1, loc=<6.9412 0 0>, rot=<0.0001 0.0018 -0.0038 -1>}
 dyn_overTie_3 {parent=dyn_overTie_2, loc=<7.9586 0 0>, rot=<-0.0003 -0.0027 -0.0152 -0.9999>}
 dyn_overTie_4 {parent=dyn_overTie_3, loc=<8.4476 0 0>, rot=<0 0 0 1>}
 dyn_spine_fur_1 {parent=spine_03, loc=<4.4906 -2.6366 -6.0298>, rot=<0.3578 -0.5624 0.1119 -0.737>}
 dyn_spine_fur_2 {parent=spine_03, loc=<0.2368 -0.102 -6.6824>, rot=<0.2571 -0.6587 -0.2571 -0.6587>}
 dyn_spine_fur_3 {parent=spine_02, loc=<5.3279 8.4047 -3.3775>, rot=<-0.185 -0.5642 -0.463 -0.6581>}
 dyn_spine_fur_6 {parent=spine_02, loc=<1.5492 8.4436 -1.8795>, rot=<-0.4635 -0.534 -0.5046 -0.4954>}
 dyn_spine_fur_9 {parent=spine_02, loc=<2.7297 -0.1594 6.097>, rot=<0.7013 -0.0901 0.6919 -0.1457>}
 dyn_spine_fur_11 {parent=spine_02, loc=<2.2273 -2.0455 -2.2555>, rot=<0.4131 -0.5739 0.366 -0.605>}
 dyn_spine_fur_12 {parent=spine_02, loc=<2.0934 2.2933 -7.0259>, rot=<0.0282 -0.7065 -0.0282 -0.7065>}
 thigh_l {parent=pelvis, loc=<-1.0855 -0.2109 -8.7053>, rot=<0.0009 0.0355 0.0155 -0.9993>}
 calf_l {parent=thigh_l, loc=<-42.5284 -0.0019 -0.7297>, rot=<0.0006 -0.0161 -0.0391 -0.9991>}
 foot_l {parent=calf_l, loc=<-46.7402 -2.5014 -0.592>, rot=<0.0427 -0.0238 0.0353 -0.9982>}
 ball_l {parent=foot_l, loc=<-3.5658 10.8066 -0.1544>, rot=<0.0001 -0.0001 -0.7063 -0.7079>}
 dyn_L_foot_fur_4 {parent=ball_l, loc=<-3.5887 1.0212 8.9469>, rot=<-0.4987 0.4989 -0.4971 -0.5052>}
 dyn_L_foot_fur_5 {parent=ball_l, loc=<3.6319 -3.9338 5.1657>, rot=<-0.6933 0.1303 -0.1252 -0.6976>}
 dyn_L_foot_fur_6 {parent=ball_l, loc=<1.8253 -2.2408 -7.0131>, rot=<-0.6249 -0.3273 0.334 -0.6251>}
 dyn_L_foot_fur_1 {parent=foot_l, loc=<-6.3044 4.6205 -9.4122>, rot=<0.0942 -0.6991 0.1021 -0.7014>}
 dyn_L_foot_fur_2 {parent=foot_l, loc=<-6.8047 -4.8046 -5.8593>, rot=<0.3499 -0.6125 0.3581 -0.6117>}
 dyn_L_foot_fur_3 {parent=foot_l, loc=<-8.3548 -2.6492 6.338>, rot=<0.6844 -0.1709 0.6895 -0.1644>}
 calf_twist_02_l {parent=calf_l, loc=<-19.1546 -1.7943 -0.2638>, rot=<0.0237 0.0095 -0.0035 -0.9997>}
 calf_twist_01_l {parent=calf_l, loc=<-35.8483 -1.9197 -0.5794>, rot=<0.0237 0.0095 -0.0035 -0.9997>}
 dyn_L_calfTwist_WPO {parent=calf_twist_01_l, loc=<0 0 0>, rot=<0 0 0 1>}
 dyn_L_leg_fur_5 {parent=calf_l, loc=<-2.9362 1.2536 -5.8621>, rot=<-0.0291 -0.6731 0.0296 -0.7384>}
 dyn_L_leg_fur_6 {parent=calf_l, loc=<-16.0287 2.3493 -8.2784>, rot=<-0.0271 -0.6242 0.0314 -0.7801>}
 dyn_L_leg_fur_7 {parent=calf_l, loc=<-30.9066 2.0978 -12.4128>, rot=<-0.0285 -0.6601 0.0301 -0.75>}
 thigh_twist_01_l {parent=thigh_l, loc=<-27.952 0 0>, rot=<0.0013 0 0 -1>}
 dyn_L_leg_fur_4 {parent=thigh_l, loc=<-30.5575 0.6767 -6.1705>, rot=<-0.0019 -0.7317 0.0015 -0.6816>}
 dyn_L_legFlap {parent=thigh_l, loc=<-21.2118 1.6998 -6.2358>, rot=<-0.6321 0.255 0.6799 -0.2705>}
 dyn_L_leg_fur_3 {parent=dyn_L_legFlap, loc=<7.7941 -0.2975 0.2588>, rot=<0.9435 -0.1932 -0.2638 -0.054>}
 dyn_L_leg_fur_2 {parent=dyn_L_legFlap, loc=<1.1418 0.0362 -5.3419>, rot=<0.963 -0.0067 -0.2693 -0.0019>}
 thigh_r {parent=pelvis, loc=<-1.0855 -0.2109 8.7053>, rot=<0.0355 -0.0009 0.9993 0.0155>}
 calf_r {parent=thigh_r, loc=<42.5284 0.0019 0.7297>, rot=<0.0006 -0.0161 -0.0391 -0.9991>}
 foot_r {parent=calf_r, loc=<46.7401 2.5014 0.592>, rot=<0.0427 -0.0238 0.0353 -0.9982>}
 ball_r {parent=foot_r, loc=<3.5658 -10.8066 0.1544>, rot=<0.0001 -0.0001 -0.7063 -0.7079>}
 dyn_R_foot_fur_4 {parent=ball_r, loc=<3.5887 -1.0212 -8.9469>, rot=<-0.4987 0.4989 -0.4971 -0.5052>}
 dyn_R_foot_fur_5 {parent=ball_r, loc=<-3.6319 3.9338 -5.1658>, rot=<-0.6933 0.1303 -0.1252 -0.6976>}
 dyn_R_foot_fur_6 {parent=ball_r, loc=<-1.8253 2.2408 7.0131>, rot=<-0.6249 -0.3273 0.334 -0.6251>}
 dyn_R_foot_fur_1 {parent=foot_r, loc=<6.3044 -4.6205 9.4122>, rot=<0.0942 -0.6991 0.1021 -0.7014>}
 dyn_R_foot_fur_2 {parent=foot_r, loc=<6.8047 4.8046 5.8593>, rot=<0.3499 -0.6125 0.3581 -0.6117>}
 dyn_R_foot_fur_3 {parent=foot_r, loc=<8.3548 2.6492 -6.338>, rot=<0.6844 -0.1709 0.6895 -0.1644>}
 calf_twist_02_r {parent=calf_r, loc=<19.1546 1.7943 0.2638>, rot=<0.0237 0.0095 -0.0035 -0.9997>}
 calf_twist_01_r {parent=calf_r, loc=<35.8483 1.9197 0.5794>, rot=<0.0237 0.0095 -0.0035 -0.9997>}
 dyn_R_calfTwist_WPO {parent=calf_twist_01_r, loc=<0 0 0>, rot=<0 0 0 1>}
 dyn_R_leg_fur_5 {parent=calf_r, loc=<2.9361 -1.2537 5.8621>, rot=<-0.0291 -0.6731 0.0296 -0.7384>}
 dyn_R_leg_fur_6 {parent=calf_r, loc=<16.0288 -2.3493 8.2783>, rot=<-0.0271 -0.6242 0.0314 -0.7801>}
 dyn_R_leg_fur_7 {parent=calf_r, loc=<30.9066 -2.0978 12.4128>, rot=<-0.0285 -0.6601 0.0301 -0.75>}
 dyn_R_leg_fur_8 {parent=calf_r, loc=<42.4425 -2.9877 12.2366>, rot=<-0.0309 -0.7201 0.0276 -0.6927>}
 thigh_twist_01_r {parent=thigh_r, loc=<27.952 0 0>, rot=<0.0013 0 0 -1>}
 dyn_R_leg_fur_4 {parent=thigh_r, loc=<30.5575 -0.6767 6.1705>, rot=<-0.0019 -0.7317 0.0015 -0.6816>}
 dyn_R_legFlap {parent=thigh_r, loc=<21.2119 -1.6998 6.2358>, rot=<0.255 0.6321 -0.2705 -0.6799>}
 dyn_R_leg_fur_2 {parent=dyn_R_legFlap, loc=<1.1418 0.0362 5.3419>, rot=<-0.0067 -0.963 0.0019 -0.2693>}
 dyn_R_leg_fur_3 {parent=dyn_R_legFlap, loc=<7.7942 -0.2975 -0.2588>, rot=<-0.1932 -0.9435 0.054 -0.2638>}
 dyn_belt {parent=pelvis, loc=<2.2512 5.72 -12.0584>, rot=<0.0127 -0.707 0.038 -0.7061>}
 dyn_L_gluteFix {parent=pelvis, loc=<-6.693 -5.7546 -9.8697>, rot=<-0.0127 -0.707 0.0127 -0.707>}
 dyn_R_gluteFix {parent=pelvis, loc=<-6.693 -5.7546 9.8696>, rot=<0.707 -0.0127 -0.707 -0.0127>}
]
