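# Physics asset for the F_MED_Mouse player body, set up for a dangling necktie with simple torso and thigh collision shapes.
# Two four-link chains, flat boxes for the over tie and thin capsules for the under tie, branch from a tie base constrained to spine_05 with twist locked.
# Spine capsules are linked by free angular constraints, while both thigh capsules have no constraints.
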PhysicsAsset /Game/Characters/Player/Female/Medium/Bodies/F_MED_Mouse/Meshes/F_MED_Mouse_Physics
bodies {
 dyn_overTie_1 [{shapeType=Box, extentX=10, extentY=7.3302, extentZ=0.6736, shapeName=dyn_overTie_1_box, center=<1.2 0 0>}]
 dyn_overTie_2 [{shapeType=Box, extentX=10, extentY=7.3302, extentZ=0.6736, shapeName=dyn_overTie_2_box, center=<1.2 0 0>}]
 dyn_overTie_3 [{shapeType=Box, extentX=10, extentY=7.3302, extentZ=0.6736, shapeName=dyn_overTie_3_box, center=<1.2 0 0>}]
 dyn_overTie_4 [{shapeType=Box, extentX=10, extentY=7.3302, extentZ=0.6736, shapeName=dyn_overTie_4_box, center=<1.2 0 0>}]
 dyn_tieBase [{shapeType=Box, extentX=10, extentY=7.3302, extentZ=0.6736, shapeName=dyn_overTie_1_box, center=<1.2 0 0>}]
 dyn_underTie_1 [{shapeType=Capsule, length=10, radius=0.3, shapeName=dyn_underTie_1_capsule, center=<2 0 0>, rotation=<rot 90 0 0>}]
 dyn_underTie_2 [{shapeType=Capsule, length=10, radius=0.7, shapeName=dyn_underTie_2_capsule, center=<2 0 0>, rotation=<rot 90 0 0>}]
 dyn_underTie_3 [{shapeType=Capsule, length=10, radius=0.7, shapeName=dyn_underTie_3_capsule, center=<2 0 0>, rotation=<rot 90 0 0>}]
 dyn_underTie_4 [{shapeType=Capsule, length=10, radius=0.7, shapeName=dyn_underTie_4_capsule, center=<2 0 0>, rotation=<rot 90 0 0>}]
 spine_01 [{shapeType=Capsule, length=10, radius=8, shapeName=spine_01_capsule, center=<0 1.1054 0>}]
 spine_03 [{shapeType=Capsule, length=10, radius=8, shapeName=spine_03_capsule, center=<-2.694 2.8074 0>}]
 spine_04 [{shapeType=Capsule, length=10, radius=9, shapeName=spine_04_capsule, center=<1.6518 2.9398 0>}]
 spine_05 [{
   shapeType=Capsule
   length=9.0377
   radius=8.1082
   shapeName=spine_05_capsule
   center=<0.1276 0.4513 0.0088>
   rotation=<rot -0.9957 -90 -0.4186>
  }]
 thigh_l [{shapeType=Capsule, length=30, radius=9, shapeName=thigh_l_capsule, center=<-22.817 -0.9689 -0.0003>, rotation=<rot 90 0 0>}]
 thigh_r [{shapeType=Capsule, length=30, radius=9, shapeName=thigh_r_capsule, center=<22.817 0.9689 0.0003>, rotation=<rot 90 0 0>}]
}
constraints [
 {
  bone1Name=spine_03
  bone2Name=spine_01
  swing1LimitDegrees=45
  swing1Motion=Free
  swing2LimitDegrees=45
  swing2Motion=Free
  twistLimitDegrees=45
  twistMotion=Free
 }
 {
  bone1Name=spine_04
  bone2Name=spine_03
  swing1LimitDegrees=45
  swing1Motion=Free
  swing2LimitDegrees=45
  swing2Motion=Free
  twistLimitDegrees=45
  twistMotion=Free
 }
 {bone1Name=dyn_overTie_1, bone2Name=dyn_tieBase, swing1LimitDegrees=45, swing2LimitDegrees=45, twistLimitDegrees=45, twistMotion=Locked}
 {bone1Name=dyn_overTie_2, bone2Name=dyn_overTie_1, swing1LimitDegrees=45, swing2LimitDegrees=45, twistLimitDegrees=45, twistMotion=Locked}
 {bone1Name=dyn_overTie_3, bone2Name=dyn_overTie_2, swing1LimitDegrees=45, swing2LimitDegrees=45, twistLimitDegrees=45, twistMotion=Locked}
 {bone1Name=dyn_overTie_4, bone2Name=dyn_overTie_3, swing1LimitDegrees=45, swing2LimitDegrees=45, twistLimitDegrees=45, twistMotion=Locked}
 {bone1Name=dyn_underTie_1, bone2Name=dyn_tieBase, swing1LimitDegrees=45, swing2LimitDegrees=45, twistLimitDegrees=45, twistMotion=Locked}
 {bone1Name=dyn_underTie_2, bone2Name=dyn_underTie_1, swing1LimitDegrees=45, swing2LimitDegrees=45, twistLimitDegrees=45, twistMotion=Locked}
 {bone1Name=dyn_underTie_3, bone2Name=dyn_underTie_2, swing1LimitDegrees=45, swing2LimitDegrees=45, twistLimitDegrees=45, twistMotion=Locked}
 {bone1Name=dyn_underTie_4, bone2Name=dyn_underTie_3, swing1LimitDegrees=45, swing2LimitDegrees=45, twistLimitDegrees=45, twistMotion=Locked}
 {bone1Name=dyn_tieBase, bone2Name=spine_05, swing1LimitDegrees=45, swing2LimitDegrees=45, twistLimitDegrees=45, twistMotion=Locked}
 {
  bone1Name=spine_05
  bone2Name=spine_04
  swing1LimitDegrees=45
  swing1Motion=Free
  swing2LimitDegrees=45
  swing2Motion=Free
  twistLimitDegrees=45
  twistMotion=Free
 }
]
