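# Post-process-style AnimBP for the female medium Mouse player body, adding procedural secondary motion on top of a copied pose.
# The pose is copied from the attached parent mesh, run through the body's Control Rig, then backpack and cape attach bones are weight-constrained to spine and clavicle bones.
# A RigidBody node feeds a long serial chain of per-bone AnimDynamics nodes that jiggle bracelets, charms, belt and fur on arms, spine, legs and feet.
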AnimBlueprint /Game/Characters/Player/Female/Medium/Bodies/F_MED_Mouse/Meshes/F_MED_Mouse_AnimBP
target_skeleton=&/F_MED_Mouse_Skeleton
graphs [
 AnimGraph {nodes [
   n1 = Control Rig!(Source=Copy Pose From Mesh() {bUseAttachedParent=true}, bAlphaBoolEnabled:bool=True) -> (Pose) {
    alphaInputType=Bool
    alphaBoolBlend {blendInTime=0.4, blendOption=Linear, blendOutTime=0.4}
    alphaScaleBiasClamp {bInterpResult=true}
    controlRigAssetReference {blueprintRigClass=&/F_MED_Mouse_CtrlRig.F_MED_Mouse_CtrlRig_C}
   }
   n2 = "Constraint - attach_backpack (PPPPP)"!(
    ComponentPose=Local To Component(LocalPose=@n1)
    Alpha=1
    ConstraintWeights_0=0.398
    ConstraintWeights_1:float=0.3306
    ConstraintWeights_2:float=0.108
    ConstraintWeights_3:float=0.105
    ConstraintWeights_4:float=0.0584
   ) -> (Pose) {
    alphaBoolBlend {blendOption=Linear}
    boneToModify {boneName=attach_backpack}
    constraintSetup [
     {targetBone {boneName=spine_04}}
     {targetBone {boneName=spine_05}}
     {targetBone {boneName=clavicle_l}}
     {targetBone {boneName=clavicle_r}}
     {targetBone {boneName=spine_03}}
    ]
    constraintWeights [1,1,1,1,1]
   }
   n3 = "Constraint - attach_cape (PPP)"!(ComponentPose=@n2, Alpha=1, ConstraintWeights_0=0.8255, ConstraintWeights_1:float=0.1033, ConstraintWeights_2:float=0.0712) -> (Pose) {
    alphaBoolBlend {blendOption=Linear}
    boneToModify {boneName=attach_cape}
    constraintSetup [{targetBone {boneName=spine_05}}, {targetBone {boneName=clavicle_r}}, {targetBone {boneName=clavicle_l}}]
    constraintWeights [1,1,1]
   }
   n4 = RigidBody(ComponentPose=@n3) {
    bClampLinearTranslationLimitToRefPose=true
    bOverrideWorldGravity=true
    overridePhysicsAsset=&/F_MED_Mouse_Physics
    simulationSpace=BaseBoneSpace
    alphaBoolBlend {blendOption=Linear}
    baseBoneRef {boneName=pelvis}
    componentLinearAccScale=<0.25 0.25 0.25>
    componentLinearVelScale=<0.3 0.3 0.3>
    overrideWorldGravity {z=-980}
   }
   n5 = Anim Dynamics - Bone: dyn_L_bracelet_1(ComponentPose=@n4, Alpha=1, ExternalForce="0,0,0") {
    angularDampingOverride=1
    angularSpringConstant=20000
    bLinearSpring=true
    bOverrideLinearDamping=true
    bUseGravityOverride=true
    gravityScale=0.35
    linearDampingOverride=1
    linearSpringConstant=20
    simulationSpace=RootRelative
    alphaBoolBlend {blendOption=Linear}
    boundBone {boneName=dyn_L_bracelet_1}
    gravityOverride {z=-980}
    physicsBodyDefinitions [{
      boundBone {boneName=dyn_L_bracelet_1}
      boxExtents=<20 20 20>
      constraintSetup {angularLimitsMax=<0 0 0>, angularLimitsMin=<0 0 0>, angularTarget=<0 0 0>, linearAxesMax=<1 0 0>}
      localJointOffset=<1 0 0>
     }]
    relativeSpaceBone {boneName=lowerarm_l}
   }
   n6 = Anim Dynamics - Bone: dyn_L_bracelet_2(ComponentPose=@n5, Alpha=1, ExternalForce="0,0,0") {
    angularDampingOverride=1
    angularSpringConstant=20000
    bLinearSpring=true
    bOverrideLinearDamping=true
    bUseGravityOverride=true
    gravityScale=0.35
    linearDampingOverride=1
    linearSpringConstant=20
    simulationSpace=RootRelative
    alphaBoolBlend {blendOption=Linear}
    boundBone {boneName=dyn_L_bracelet_2}
    gravityOverride {z=-980}
    physicsBodyDefinitions [{
      boundBone {boneName=dyn_L_bracelet_2}
      boxExtents=<20 20 20>
      constraintSetup {angularLimitsMax=<0 0 0>, angularLimitsMin=<0 0 0>, angularTarget=<0 0 0>, linearAxesMax=<0.5 0 0>}
      localJointOffset=<1 0 0>
     }]
    relativeSpaceBone {boneName=lowerarm_l}
   }
   n7 = Anim Dynamics - Bone: dyn_R_bracelet_1(ComponentPose=@n6, Alpha=1, ExternalForce="0,0,0") {
    angularDampingOverride=1
    angularSpringConstant=20000
    bLinearSpring=true
    bOverrideLinearDamping=true
    bUseGravityOverride=true
    gravityScale=0.35
    linearDampingOverride=1
    linearSpringConstant=20
    simulationSpace=RootRelative
    alphaBoolBlend {blendOption=Linear}
    boundBone {boneName=dyn_R_bracelet_1}
    gravityOverride {z=-980}
    physicsBodyDefinitions [{
      boundBone {boneName=dyn_R_bracelet_1}
      boxExtents=<20 20 20>
      constraintSetup {angularLimitsMax=<0 0 0>, angularLimitsMin=<0 0 0>, angularTarget=<0 0 0>, linearAxesMax=<1 0 0>}
      localJointOffset=<1 0 0>
     }]
    relativeSpaceBone {boneName=lowerarm_r}
   }
   n8 = Anim Dynamics - Bone: dyn_R_bracelet_2(ComponentPose=@n7, Alpha=1, ExternalForce="0,0,0") {
    angularDampingOverride=1
    angularSpringConstant=20000
    bLinearSpring=true
    bOverrideLinearDamping=true
    bUseGravityOverride=true
    gravityScale=0.35
    linearDampingOverride=1
    linearSpringConstant=20
    simulationSpace=RootRelative
    alphaBoolBlend {blendOption=Linear}
    boundBone {boneName=dyn_R_bracelet_2}
    gravityOverride {z=-980}
    physicsBodyDefinitions [{
      boundBone {boneName=dyn_R_bracelet_2}
      boxExtents=<20 20 20>
      constraintSetup {angularLimitsMax=<0 0 0>, angularLimitsMin=<0 0 0>, angularTarget=<0 0 0>, linearAxesMax=<0.5 0 0>}
      localJointOffset=<1 0 0>
     }]
    relativeSpaceBone {boneName=lowerarm_r}
   }
   n9 = Anim Dynamics - Bone: dyn_L_bracelet_charm_A(ComponentPose=@n8, Alpha=1, ExternalForce="0,0,0") {
    angularDampingOverride=1
    angularSpringConstant=20
    bOverrideAngularDamping=true
    bUseGravityOverride=true
    gravityScale=0.35
    simulationSpace=RootRelative
    alphaBoolBlend {blendOption=Linear}
    boundBone {boneName=dyn_L_bracelet_charm_A}
    componentLinearAccScale=<1 1 1>
    componentLinearVelScale=<1 1 1>
    gravityOverride {z=-980}
    physicsBodyDefinitions [{
      boundBone {boneName=dyn_L_bracelet_charm_A}
      boxExtents=<2 2 2>
      constraintSetup {angularLimitsMax=<0 150 0>, angularLimitsMin=<0 -10 0>, angularTarget=<1 0 0>}
      localJointOffset=<0.5 0 0>
     }]
    relativeSpaceBone {boneName=spine_01}
   }
   n10 = Anim Dynamics - Bone: dyn_L_bracelet_charm_B(ComponentPose=@n9, Alpha=1, ExternalForce="0,0,0") {
    angularDampingOverride=1
    angularSpringConstant=20
    bOverrideAngularDamping=true
    bUseGravityOverride=true
    gravityScale=0.35
    simulationSpace=RootRelative
    alphaBoolBlend {blendOption=Linear}
    boundBone {boneName=dyn_L_bracelet_charm_B}
    componentLinearAccScale=<1 1 1>
    componentLinearVelScale=<1 1 1>
    gravityOverride {z=-980}
    physicsBodyDefinitions [{
      boundBone {boneName=dyn_L_bracelet_charm_B}
      boxExtents=<2 2 2>
      constraintSetup {angularLimitsMax=<0 150 0>, angularLimitsMin=<0 -10 0>, angularTarget=<1 0 0>}
      localJointOffset=<0.5 0 0>
     }]
    relativeSpaceBone {boneName=spine_01}
   }
   n11 = Anim Dynamics - Bone: dyn_L_bracelet_charm_C(ComponentPose=@n10, Alpha=1, ExternalForce="0,0,0") {
    angularDampingOverride=1
    angularSpringConstant=20
    bOverrideAngularDamping=true
    bUseGravityOverride=true
    gravityScale=0.35
    simulationSpace=RootRelative
    alphaBoolBlend {blendOption=Linear}
    boundBone {boneName=dyn_L_bracelet_charm_C}
    componentLinearAccScale=<1 1 1>
    componentLinearVelScale=<1 1 1>
    gravityOverride {z=-980}
    physicsBodyDefinitions [{
      boundBone {boneName=dyn_L_bracelet_charm_C}
      boxExtents=<2 2 2>
      constraintSetup {angularLimitsMax=<0 150 0>, angularLimitsMin=<0 -10 0>, angularTarget=<1 0 0>}
      localJointOffset=<0.5 0 0>
     }]
    relativeSpaceBone {boneName=spine_01}
   }
   n12 = Anim Dynamics - Bone: dyn_L_bracelet_charm_D(ComponentPose=@n11, Alpha=1, ExternalForce="0,0,0") {
    angularDampingOverride=1
    angularSpringConstant=20
    bOverrideAngularDamping=true
    bUseGravityOverride=true
    gravityScale=0.35
    simulationSpace=RootRelative
    alphaBoolBlend {blendOption=Linear}
    boundBone {boneName=dyn_L_bracelet_charm_D}
    componentLinearAccScale=<1 1 1>
    componentLinearVelScale=<1 1 1>
    gravityOverride {z=-980}
    physicsBodyDefinitions [{
      boundBone {boneName=dyn_L_bracelet_charm_D}
      boxExtents=<2 2 2>
      constraintSetup {angularLimitsMax=<0 150 0>, angularLimitsMin=<0 -90 0>, angularTarget=<1 0 0>}
      localJointOffset=<0.5 0 0>
     }]
    relativeSpaceBone {boneName=spine_01}
   }
   n13 = Anim Dynamics - Bone: dyn_R_bracelet_charm_A(ComponentPose=@n12, Alpha=1, ExternalForce="0,0,0") {
    angularDampingOverride=1
    angularSpringConstant=20
    bOverrideAngularDamping=true
    bUseGravityOverride=true
    gravityScale=0.35
    simulationSpace=RootRelative
    alphaBoolBlend {blendOption=Linear}
    boundBone {boneName=dyn_R_bracelet_charm_A}
    componentLinearAccScale=<1 1 1>
    componentLinearVelScale=<1 1 1>
    gravityOverride {z=-980}
    physicsBodyDefinitions [{
      boundBone {boneName=dyn_R_bracelet_charm_A}
      boxExtents=<2 2 2>
      constraintSetup {angularLimitsMax=<0 150 0>, angularLimitsMin=<0 -10 0>, angularTarget=<1 0 0>}
      localJointOffset=<0.5 0 0>
     }]
    relativeSpaceBone {boneName=spine_01}
   }
   n14 = Anim Dynamics - Bone: dyn_R_bracelet_charm_B(ComponentPose=@n13, Alpha=1, ExternalForce="0,0,0") {
    angularDampingOverride=1
    angularSpringConstant=20
    bOverrideAngularDamping=true
    bUseGravityOverride=true
    gravityScale=0.35
    simulationSpace=RootRelative
    alphaBoolBlend {blendOption=Linear}
    boundBone {boneName=dyn_R_bracelet_charm_B}
    componentLinearAccScale=<1 1 1>
    componentLinearVelScale=<1 1 1>
    gravityOverride {z=-980}
    physicsBodyDefinitions [{
      boundBone {boneName=dyn_R_bracelet_charm_B}
      boxExtents=<2 2 2>
      constraintSetup {angularLimitsMax=<0 150 0>, angularLimitsMin=<0 -10 0>, angularTarget=<1 0 0>}
      localJointOffset=<0.5 0 0>
     }]
    relativeSpaceBone {boneName=spine_01}
   }
   n15 = Anim Dynamics - Bone: dyn_R_bracelet_charm_C(ComponentPose=@n14, Alpha=1, ExternalForce="0,0,0") {
    angularDampingOverride=1
    angularSpringConstant=20
    bOverrideAngularDamping=true
    bUseGravityOverride=true
    gravityScale=0.35
    simulationSpace=RootRelative
    alphaBoolBlend {blendOption=Linear}
    boundBone {boneName=dyn_R_bracelet_charm_C}
    componentLinearAccScale=<1 1 1>
    componentLinearVelScale=<1 1 1>
    gravityOverride {z=-980}
    physicsBodyDefinitions [{
      boundBone {boneName=dyn_R_bracelet_charm_C}
      boxExtents=<2 2 2>
      constraintSetup {angularLimitsMax=<0 150 0>, angularLimitsMin=<0 -10 0>, angularTarget=<1 0 0>}
      localJointOffset=<0.5 0 0>
     }]
    relativeSpaceBone {boneName=spine_01}
   }
   n16 = Anim Dynamics - Bone: dyn_R_bracelet_charm_D(ComponentPose=@n15, Alpha=1, ExternalForce="0,0,0") {
    angularDampingOverride=1
    angularSpringConstant=20
    bOverrideAngularDamping=true
    bUseGravityOverride=true
    gravityScale=0.35
    simulationSpace=RootRelative
    alphaBoolBlend {blendOption=Linear}
    boundBone {boneName=dyn_R_bracelet_charm_D}
    componentLinearAccScale=<1 1 1>
    componentLinearVelScale=<1 1 1>
    gravityOverride {z=-980}
    physicsBodyDefinitions [{
      boundBone {boneName=dyn_R_bracelet_charm_D}
      boxExtents=<2 2 2>
      constraintSetup {angularLimitsMax=<0 150 0>, angularLimitsMin=<0 -90 0>, angularTarget=<1 0 0>}
      localJointOffset=<0.5 0 0>
     }]
    relativeSpaceBone {boneName=spine_01}
   }
   n17 = Anim Dynamics - Bone: dyn_belt(ComponentPose=@n16, Alpha=1, ExternalForce="0,0,0") {
    angularDampingOverride=1
    angularSpringConstant=2500
    bAngularSpring=true
    bOverrideAngularDamping=true
    bUseGravityOverride=true
    gravityScale=0.35
    simulationSpace=RootRelative
    alphaBoolBlend {blendOption=Linear}
    boundBone {boneName=dyn_belt}
    componentLinearAccScale=<1 1 1>
    componentLinearVelScale=<1 1 1>
    gravityOverride {z=-980}
    physicsBodyDefinitions [{
      boundBone {boneName=dyn_belt}
      boxExtents=<20 20 20>
      constraintSetup {angularLimitsMax=<0 5 30>, angularLimitsMin=<0 -5 -30>, angularTarget=<1 0 0>}
      localJointOffset=<2 0 0>
     }]
    relativeSpaceBone {boneName=pelvis}
   }
   n18 = Anim Dynamics - Bone: dyn_L_arm_fur_1(ComponentPose=@n17, Alpha=1, ExternalForce="0,0,0") {
    angularDampingOverride=0.9999
    angularSpringConstant=5000
    bAngularSpring=true
    bOverrideAngularDamping=true
    bUseGravityOverride=true
    gravityScale=0.35
    simulationSpace=BoneRelative
    alphaBoolBlend {blendOption=Linear}
    boundBone {boneName=dyn_L_arm_fur_1}
    componentLinearAccScale=<1 1 1>
    componentLinearVelScale=<1 1 1>
    gravityOverride {z=-980}
    physicsBodyDefinitions [{
      boundBone {boneName=dyn_L_arm_fur_1}
      boxExtents=<10 10 10>
      constraintSetup {angularLimitsMax=<0 30 30>, angularLimitsMin=<0 -30 -30>, angularTarget=<1 0 0>}
      localJointOffset=<1 0 0>
     }]
    relativeSpaceBone {boneName=spine_05}
   }
   n19 = Anim Dynamics - Bone: dyn_L_arm_fur_2(ComponentPose=@n18, Alpha=1, ExternalForce="0,0,0") {
    angularDampingOverride=0.9999
    angularSpringConstant=5000
    bAngularSpring=true
    bOverrideAngularDamping=true
    bUseGravityOverride=true
    gravityScale=0.35
    simulationSpace=BoneRelative
    alphaBoolBlend {blendOption=Linear}
    boundBone {boneName=dyn_L_arm_fur_2}
    componentLinearAccScale=<1 1 1>
    componentLinearVelScale=<1 1 1>
    gravityOverride {z=-980}
    physicsBodyDefinitions [{
      boundBone {boneName=dyn_L_arm_fur_2}
      boxExtents=<10 10 10>
      constraintSetup {angularLimitsMax=<0 30 30>, angularLimitsMin=<0 -30 -30>, angularTarget=<1 0 0>}
      localJointOffset=<1 0 0>
     }]
    relativeSpaceBone {boneName=spine_05}
   }
   n20 = Anim Dynamics - Bone: dyn_L_arm_fur_3(ComponentPose=@n19, Alpha=1, ExternalForce="0,0,0") {
    angularDampingOverride=0.9999
    angularSpringConstant=5000
    bAngularSpring=true
    bOverrideAngularDamping=true
    bUseGravityOverride=true
    gravityScale=0.35
    simulationSpace=BoneRelative
    alphaBoolBlend {blendOption=Linear}
    boundBone {boneName=dyn_L_arm_fur_3}
    componentLinearAccScale=<1 1 1>
    componentLinearVelScale=<1 1 1>
    gravityOverride {z=-980}
    physicsBodyDefinitions [{
      boundBone {boneName=dyn_L_arm_fur_3}
      boxExtents=<10 10 10>
      constraintSetup {angularLimitsMax=<0 30 30>, angularLimitsMin=<0 -30 -30>, angularTarget=<1 0 0>}
      localJointOffset=<1 0 0>
     }]
    relativeSpaceBone {boneName=spine_05}
   }
   n21 = Anim Dynamics - Bone: dyn_L_arm_fur_4(ComponentPose=@n20, Alpha=1, ExternalForce="0,0,0") {
    angularDampingOverride=0.9999
    angularSpringConstant=5000
    bAngularSpring=true
    bOverrideAngularDamping=true
    bUseGravityOverride=true
    gravityScale=0.35
    simulationSpace=BoneRelative
    alphaBoolBlend {blendOption=Linear}
    boundBone {boneName=dyn_L_arm_fur_4}
    componentLinearAccScale=<1 1 1>
    componentLinearVelScale=<1 1 1>
    gravityOverride {z=-980}
    physicsBodyDefinitions [{
      boundBone {boneName=dyn_L_arm_fur_4}
      boxExtents=<10 10 10>
      constraintSetup {angularLimitsMax=<0 30 30>, angularLimitsMin=<0 -30 -30>, angularTarget=<1 0 0>}
      localJointOffset=<1 0 0>
     }]
    relativeSpaceBone {boneName=spine_05}
   }
   n22 = Anim Dynamics - Bone: dyn_L_arm_fur_5(ComponentPose=@n21, Alpha=1, ExternalForce="0,0,0") {
    angularDampingOverride=0.9999
    angularSpringConstant=5000
    bAngularSpring=true
    bOverrideAngularDamping=true
    bUseGravityOverride=true
    gravityScale=0.35
    simulationSpace=BoneRelative
    alphaBoolBlend {blendOption=Linear}
    boundBone {boneName=dyn_L_arm_fur_5}
    componentLinearAccScale=<1 1 1>
    componentLinearVelScale=<1 1 1>
    gravityOverride {z=-980}
    physicsBodyDefinitions [{
      boundBone {boneName=dyn_L_arm_fur_5}
      boxExtents=<10 10 10>
      constraintSetup {angularLimitsMax=<0 30 30>, angularLimitsMin=<0 -30 -30>, angularTarget=<1 0 0>}
      localJointOffset=<1 0 0>
     }]
    relativeSpaceBone {boneName=spine_05}
   }
   n23 = Anim Dynamics - Bone: dyn_R_arm_fur_1(ComponentPose=@n22, Alpha=1, ExternalForce="0,0,0") {
    angularDampingOverride=0.9999
    angularSpringConstant=5000
    bAngularSpring=true
    bOverrideAngularDamping=true
    bUseGravityOverride=true
    gravityScale=0.35
    simulationSpace=BoneRelative
    alphaBoolBlend {blendOption=Linear}
    boundBone {boneName=dyn_R_arm_fur_1}
    componentLinearAccScale=<1 1 1>
    componentLinearVelScale=<1 1 1>
    gravityOverride {z=-980}
    physicsBodyDefinitions [{
      boundBone {boneName=dyn_R_arm_fur_1}
      boxExtents=<10 10 10>
      constraintSetup {angularLimitsMax=<0 30 30>, angularLimitsMin=<0 -30 -30>, angularTarget=<1 0 0>}
      localJointOffset=<1 0 0>
     }]
    relativeSpaceBone {boneName=spine_05}
   }
   n24 = Anim Dynamics - Bone: dyn_R_arm_fur_2(ComponentPose=@n23, Alpha=1, ExternalForce="0,0,0") {
    angularDampingOverride=0.9999
    angularSpringConstant=5000
    bAngularSpring=true
    bOverrideAngularDamping=true
    bUseGravityOverride=true
    gravityScale=0.35
    simulationSpace=BoneRelative
    alphaBoolBlend {blendOption=Linear}
    boundBone {boneName=dyn_R_arm_fur_2}
    componentLinearAccScale=<1 1 1>
    componentLinearVelScale=<1 1 1>
    gravityOverride {z=-980}
    physicsBodyDefinitions [{
      boundBone {boneName=dyn_R_arm_fur_2}
      boxExtents=<10 10 10>
      constraintSetup {angularLimitsMax=<0 30 30>, angularLimitsMin=<0 -30 -30>, angularTarget=<1 0 0>}
      localJointOffset=<1 0 0>
     }]
    relativeSpaceBone {boneName=spine_05}
   }
   n25 = Anim Dynamics - Bone: dyn_R_arm_fur_3(ComponentPose=@n24, Alpha=1, ExternalForce="0,0,0") {
    angularDampingOverride=0.9999
    angularSpringConstant=5000
    bAngularSpring=true
    bOverrideAngularDamping=true
    bUseGravityOverride=true
    gravityScale=0.35
    simulationSpace=BoneRelative
    alphaBoolBlend {blendOption=Linear}
    boundBone {boneName=dyn_R_arm_fur_3}
    componentLinearAccScale=<1 1 1>
    componentLinearVelScale=<1 1 1>
    gravityOverride {z=-980}
    physicsBodyDefinitions [{
      boundBone {boneName=dyn_R_arm_fur_3}
      boxExtents=<10 10 10>
      constraintSetup {angularLimitsMax=<0 30 30>, angularLimitsMin=<0 -30 -30>, angularTarget=<1 0 0>}
      localJointOffset=<1 0 0>
     }]
    relativeSpaceBone {boneName=spine_05}
   }
   n26 = Anim Dynamics - Bone: dyn_R_arm_fur_4(ComponentPose=@n25, Alpha=1, ExternalForce="0,0,0") {
    angularDampingOverride=0.9999
    angularSpringConstant=5000
    bAngularSpring=true
    bOverrideAngularDamping=true
    bUseGravityOverride=true
    gravityScale=0.35
    simulationSpace=BoneRelative
    alphaBoolBlend {blendOption=Linear}
    boundBone {boneName=dyn_R_arm_fur_4}
    componentLinearAccScale=<1 1 1>
    componentLinearVelScale=<1 1 1>
    gravityOverride {z=-980}
    physicsBodyDefinitions [{
      boundBone {boneName=dyn_R_arm_fur_4}
      boxExtents=<10 10 10>
      constraintSetup {angularLimitsMax=<0 30 30>, angularLimitsMin=<0 -30 -30>, angularTarget=<1 0 0>}
      localJointOffset=<1 0 0>
     }]
    relativeSpaceBone {boneName=spine_05}
   }
   n27 = Anim Dynamics - Bone: dyn_R_arm_fur_5(ComponentPose=@n26, Alpha=1, ExternalForce="0,0,0") {
    angularDampingOverride=0.9999
    angularSpringConstant=5000
    bAngularSpring=true
    bOverrideAngularDamping=true
    bUseGravityOverride=true
    gravityScale=0.35
    simulationSpace=BoneRelative
    alphaBoolBlend {blendOption=Linear}
    boundBone {boneName=dyn_R_arm_fur_5}
    componentLinearAccScale=<1 1 1>
    componentLinearVelScale=<1 1 1>
    gravityOverride {z=-980}
    physicsBodyDefinitions [{
      boundBone {boneName=dyn_R_arm_fur_5}
      boxExtents=<10 10 10>
      constraintSetup {angularLimitsMax=<0 30 30>, angularLimitsMin=<0 -30 -30>, angularTarget=<1 0 0>}
      localJointOffset=<1 0 0>
     }]
    relativeSpaceBone {boneName=spine_05}
   }
   n28 = Anim Dynamics - Bone: dyn_R_arm_fur_6(ComponentPose=@n27, Alpha=1, ExternalForce="0,0,0") {
    angularDampingOverride=0.9999
    angularSpringConstant=5000
    bAngularSpring=true
    bOverrideAngularDamping=true
    bUseGravityOverride=true
    gravityScale=0.35
    simulationSpace=BoneRelative
    alphaBoolBlend {blendOption=Linear}
    boundBone {boneName=dyn_R_arm_fur_6}
    componentLinearAccScale=<1 1 1>
    componentLinearVelScale=<1 1 1>
    gravityOverride {z=-980}
    physicsBodyDefinitions [{
      boundBone {boneName=dyn_R_arm_fur_6}
      boxExtents=<10 10 10>
      constraintSetup {angularLimitsMax=<0 30 30>, angularLimitsMin=<0 -30 -30>, angularTarget=<1 0 0>}
      localJointOffset=<1 0 0>
     }]
    relativeSpaceBone {boneName=spine_05}
   }
   n29 = Anim Dynamics - Bone: dyn_R_arm_fur_7(ComponentPose=@n28, Alpha=1, ExternalForce="0,0,0") {
    angularDampingOverride=0.9999
    angularSpringConstant=5000
    bAngularSpring=true
    bOverrideAngularDamping=true
    bUseGravityOverride=true
    gravityScale=0.35
    simulationSpace=BoneRelative
    alphaBoolBlend {blendOption=Linear}
    boundBone {boneName=dyn_R_arm_fur_7}
    componentLinearAccScale=<1 1 1>
    componentLinearVelScale=<1 1 1>
    gravityOverride {z=-980}
    physicsBodyDefinitions [{
      boundBone {boneName=dyn_R_arm_fur_7}
      boxExtents=<10 10 10>
      constraintSetup {angularLimitsMax=<0 30 30>, angularLimitsMin=<0 -30 -30>, angularTarget=<1 0 0>}
      localJointOffset=<1 0 0>
     }]
    relativeSpaceBone {boneName=spine_05}
   }
   n30 = Anim Dynamics - Bone: dyn_spine_fur_1(ComponentPose=@n29, Alpha=1, ExternalForce="0,0,0") {
    angularDampingOverride=0.9999
    angularSpringConstant=5000
    bAngularSpring=true
    bOverrideAngularDamping=true
    bUseGravityOverride=true
    gravityScale=0.35
    simulationSpace=BoneRelative
    alphaBoolBlend {blendOption=Linear}
    boundBone {boneName=dyn_spine_fur_1}
    componentLinearAccScale=<1 1 1>
    componentLinearVelScale=<1 1 1>
    gravityOverride {z=-980}
    physicsBodyDefinitions [{
      boundBone {boneName=dyn_spine_fur_1}
      boxExtents=<10 10 10>
      constraintSetup {angularLimitsMax=<0 30 30>, angularLimitsMin=<0 -30 -30>, angularTarget=<1 0 0>}
      localJointOffset=<1 0 0>
     }]
    relativeSpaceBone {boneName=pelvis}
   }
   n31 = Anim Dynamics - Bone: dyn_spine_fur_2(ComponentPose=@n30, Alpha=1, ExternalForce="0,0,0") {
    angularDampingOverride=0.9999
    angularSpringConstant=5000
    bAngularSpring=true
    bOverrideAngularDamping=true
    bUseGravityOverride=true
    gravityScale=0.35
    simulationSpace=BoneRelative
    alphaBoolBlend {blendOption=Linear}
    boundBone {boneName=dyn_spine_fur_2}
    componentLinearAccScale=<1 1 1>
    componentLinearVelScale=<1 1 1>
    gravityOverride {z=-980}
    physicsBodyDefinitions [{
      boundBone {boneName=dyn_spine_fur_2}
      boxExtents=<10 10 10>
      constraintSetup {angularLimitsMax=<0 30 30>, angularLimitsMin=<0 -30 -30>, angularTarget=<1 0 0>}
      localJointOffset=<1 0 0>
     }]
    relativeSpaceBone {boneName=pelvis}
   }
   n32 = Anim Dynamics - Bone: dyn_spine_fur_3(ComponentPose=@n31, Alpha=1, ExternalForce="0,0,0") {
    angularDampingOverride=0.9999
    angularSpringConstant=5000
    bAngularSpring=true
    bOverrideAngularDamping=true
    bUseGravityOverride=true
    gravityScale=0.35
    simulationSpace=BoneRelative
    alphaBoolBlend {blendOption=Linear}
    boundBone {boneName=dyn_spine_fur_3}
    componentLinearAccScale=<1 1 1>
    componentLinearVelScale=<1 1 1>
    gravityOverride {z=-980}
    physicsBodyDefinitions [{
      boundBone {boneName=dyn_spine_fur_3}
      boxExtents=<10 10 10>
      constraintSetup {angularLimitsMax=<0 30 30>, angularLimitsMin=<0 -30 -30>, angularTarget=<1 0 0>}
      localJointOffset=<1 0 0>
     }]
    relativeSpaceBone {boneName=pelvis}
   }
   n33 = Anim Dynamics - Bone: dyn_spine_fur_6(ComponentPose=@n32, Alpha=1, ExternalForce="0,0,0") {
    angularDampingOverride=0.9999
    angularSpringConstant=5000
    bAngularSpring=true
    bOverrideAngularDamping=true
    bUseGravityOverride=true
    gravityScale=0.35
    simulationSpace=BoneRelative
    alphaBoolBlend {blendOption=Linear}
    boundBone {boneName=dyn_spine_fur_6}
    componentLinearAccScale=<1 1 1>
    componentLinearVelScale=<1 1 1>
    gravityOverride {z=-980}
    physicsBodyDefinitions [{
      boundBone {boneName=dyn_spine_fur_6}
      boxExtents=<10 10 10>
      constraintSetup {angularLimitsMax=<0 30 30>, angularLimitsMin=<0 -30 -30>, angularTarget=<1 0 0>}
      localJointOffset=<1 0 0>
     }]
    relativeSpaceBone {boneName=pelvis}
   }
   n34 = Anim Dynamics - Bone: dyn_spine_fur_9(ComponentPose=@n33, Alpha=1, ExternalForce="0,0,0") {
    angularDampingOverride=0.9999
    angularSpringConstant=5000
    bAngularSpring=true
    bOverrideAngularDamping=true
    bUseGravityOverride=true
    gravityScale=0.35
    simulationSpace=BoneRelative
    alphaBoolBlend {blendOption=Linear}
    boundBone {boneName=dyn_spine_fur_9}
    componentLinearAccScale=<1 1 1>
    componentLinearVelScale=<1 1 1>
    gravityOverride {z=-980}
    physicsBodyDefinitions [{
      boundBone {boneName=dyn_spine_fur_9}
      boxExtents=<10 10 10>
      constraintSetup {angularLimitsMax=<0 30 30>, angularLimitsMin=<0 -30 -30>, angularTarget=<1 0 0>}
      localJointOffset=<1 0 0>
     }]
    relativeSpaceBone {boneName=pelvis}
   }
   n35 = Anim Dynamics - Bone: dyn_spine_fur_11(ComponentPose=@n34, Alpha=1, ExternalForce="0,0,0") {
    angularDampingOverride=0.9999
    angularSpringConstant=5000
    bAngularSpring=true
    bOverrideAngularDamping=true
    bUseGravityOverride=true
    gravityScale=0.35
    simulationSpace=BoneRelative
    alphaBoolBlend {blendOption=Linear}
    boundBone {boneName=dyn_spine_fur_11}
    componentLinearAccScale=<1 1 1>
    componentLinearVelScale=<1 1 1>
    gravityOverride {z=-980}
    physicsBodyDefinitions [{
      boundBone {boneName=dyn_spine_fur_11}
      boxExtents=<10 10 10>
      constraintSetup {angularLimitsMax=<0 30 30>, angularLimitsMin=<0 -30 -30>, angularTarget=<1 0 0>}
      localJointOffset=<1 0 0>
     }]
    relativeSpaceBone {boneName=pelvis}
   }
   n36 = Anim Dynamics - Bone: dyn_spine_fur_12(ComponentPose=@n35, Alpha=1, ExternalForce="0,0,0") {
    angularDampingOverride=0.9999
    angularSpringConstant=5000
    bAngularSpring=true
    bOverrideAngularDamping=true
    bUseGravityOverride=true
    gravityScale=0.35
    simulationSpace=BoneRelative
    alphaBoolBlend {blendOption=Linear}
    boundBone {boneName=dyn_spine_fur_12}
    componentLinearAccScale=<1 1 1>
    componentLinearVelScale=<1 1 1>
    gravityOverride {z=-980}
    physicsBodyDefinitions [{
      boundBone {boneName=dyn_spine_fur_12}
      boxExtents=<10 10 10>
      constraintSetup {angularLimitsMax=<0 30 30>, angularLimitsMin=<0 -30 -30>, angularTarget=<1 0 0>}
      localJointOffset=<1 0 0>
     }]
    relativeSpaceBone {boneName=pelvis}
   }
   n37 = Anim Dynamics - Bone: dyn_L_leg_fur_2(ComponentPose=@n36, Alpha=1, ExternalForce="0,0,0") {
    angularDampingOverride=0.9999
    angularSpringConstant=5000
    bAngularSpring=true
    bOverrideAngularDamping=true
    bUseGravityOverride=true
    gravityScale=0.35
    simulationSpace=BoneRelative
    alphaBoolBlend {blendOption=Linear}
    boundBone {boneName=dyn_L_leg_fur_2}
    componentLinearAccScale=<1 1 1>
    componentLinearVelScale=<1 1 1>
    gravityOverride {z=-980}
    physicsBodyDefinitions [{
      boundBone {boneName=dyn_L_leg_fur_2}
      boxExtents=<10 10 10>
      constraintSetup {angularLimitsMax=<0 30 30>, angularLimitsMin=<0 -30 -30>, angularTarget=<1 0 0>}
      localJointOffset=<1 0 0>
     }]
    relativeSpaceBone {boneName=pelvis}
   }
   n38 = Anim Dynamics - Bone: dyn_L_leg_fur_3(ComponentPose=@n37, Alpha=1, ExternalForce="0,0,0") {
    angularDampingOverride=0.9999
    angularSpringConstant=5000
    bAngularSpring=true
    bOverrideAngularDamping=true
    bUseGravityOverride=true
    gravityScale=0.35
    simulationSpace=BoneRelative
    alphaBoolBlend {blendOption=Linear}
    boundBone {boneName=dyn_L_leg_fur_3}
    componentLinearAccScale=<1 1 1>
    componentLinearVelScale=<1 1 1>
    gravityOverride {z=-980}
    physicsBodyDefinitions [{
      boundBone {boneName=dyn_L_leg_fur_3}
      boxExtents=<10 10 10>
      constraintSetup {angularLimitsMax=<0 30 0>, angularLimitsMin=<0 -30 0>, angularTarget=<1 0 0>}
      localJointOffset=<1 0 0>
     }]
    relativeSpaceBone {boneName=pelvis}
   }
   n39 = Anim Dynamics - Bone: dyn_L_leg_fur_4(ComponentPose=@n38, Alpha=1, ExternalForce="0,0,0") {
    angularDampingOverride=0.9999
    angularSpringConstant=5000
    bAngularSpring=true
    bOverrideAngularDamping=true
    bUseGravityOverride=true
    gravityScale=0.35
    simulationSpace=BoneRelative
    alphaBoolBlend {blendOption=Linear}
    boundBone {boneName=dyn_L_leg_fur_4}
    componentLinearAccScale=<1 1 1>
    componentLinearVelScale=<1 1 1>
    gravityOverride {z=-980}
    physicsBodyDefinitions [{
      boundBone {boneName=dyn_L_leg_fur_4}
      boxExtents=<10 10 10>
      constraintSetup {angularLimitsMax=<0 30 30>, angularLimitsMin=<0 -30 -30>, angularTarget=<1 0 0>}
      localJointOffset=<1 0 0>
     }]
    relativeSpaceBone {boneName=pelvis}
   }
   n40 = Anim Dynamics - Bone: dyn_L_leg_fur_5(ComponentPose=@n39, Alpha=1, ExternalForce="0,0,0") {
    angularDampingOverride=0.9999
    angularSpringConstant=5000
    bAngularSpring=true
    bOverrideAngularDamping=true
    bUseGravityOverride=true
    gravityScale=0.35
    simulationSpace=BoneRelative
    alphaBoolBlend {blendOption=Linear}
    boundBone {boneName=dyn_L_leg_fur_5}
    componentLinearAccScale=<1 1 1>
    componentLinearVelScale=<1 1 1>
    gravityOverride {z=-980}
    physicsBodyDefinitions [{
      boundBone {boneName=dyn_L_leg_fur_5}
      boxExtents=<10 10 10>
      constraintSetup {angularLimitsMax=<0 30 30>, angularLimitsMin=<0 -30 -30>, angularTarget=<1 0 0>}
      localJointOffset=<1 0 0>
     }]
    relativeSpaceBone {boneName=pelvis}
   }
   n41 = Anim Dynamics - Bone: dyn_L_leg_fur_6(ComponentPose=@n40, Alpha=1, ExternalForce="0,0,0") {
    angularDampingOverride=0.9999
    angularSpringConstant=5000
    bAngularSpring=true
    bOverrideAngularDamping=true
    bUseGravityOverride=true
    gravityScale=0.35
    simulationSpace=BoneRelative
    alphaBoolBlend {blendOption=Linear}
    boundBone {boneName=dyn_L_leg_fur_6}
    componentLinearAccScale=<1 1 1>
    componentLinearVelScale=<1 1 1>
    gravityOverride {z=-980}
    physicsBodyDefinitions [{
      boundBone {boneName=dyn_L_leg_fur_6}
      boxExtents=<10 10 10>
      constraintSetup {angularLimitsMax=<0 30 30>, angularLimitsMin=<0 -30 -30>, angularTarget=<1 0 0>}
      localJointOffset=<1 0 0>
     }]
    relativeSpaceBone {boneName=pelvis}
   }
   n42 = Anim Dynamics - Bone: dyn_L_leg_fur_7(ComponentPose=@n41, Alpha=1, ExternalForce="0,0,0") {
    angularDampingOverride=0.9999
    angularSpringConstant=5000
    bAngularSpring=true
    bOverrideAngularDamping=true
    bUseGravityOverride=true
    gravityScale=0.35
    simulationSpace=BoneRelative
    alphaBoolBlend {blendOption=Linear}
    boundBone {boneName=dyn_L_leg_fur_7}
    componentLinearAccScale=<1 1 1>
    componentLinearVelScale=<1 1 1>
    gravityOverride {z=-980}
    physicsBodyDefinitions [{
      boundBone {boneName=dyn_L_leg_fur_7}
      boxExtents=<10 10 10>
      constraintSetup {angularLimitsMax=<0 30 30>, angularLimitsMin=<0 -30 -30>, angularTarget=<1 0 0>}
      localJointOffset=<1 0 0>
     }]
    relativeSpaceBone {boneName=pelvis}
   }
   n43 = Anim Dynamics - Bone: dyn_R_leg_fur_2(ComponentPose=@n42, Alpha=1, ExternalForce="0,0,0") {
    angularDampingOverride=0.9999
    angularSpringConstant=5000
    bAngularSpring=true
    bOverrideAngularDamping=true
    bUseGravityOverride=true
    gravityScale=0.35
    simulationSpace=BoneRelative
    alphaBoolBlend {blendOption=Linear}
    boundBone {boneName=dyn_R_leg_fur_2}
    componentLinearAccScale=<1 1 1>
    componentLinearVelScale=<1 1 1>
    gravityOverride {z=-980}
    physicsBodyDefinitions [{
      boundBone {boneName=dyn_R_leg_fur_2}
      boxExtents=<10 10 10>
      constraintSetup {angularLimitsMax=<0 30 30>, angularLimitsMin=<0 -30 -30>, angularTarget=<1 0 0>}
      localJointOffset=<1 0 0>
     }]
    relativeSpaceBone {boneName=pelvis}
   }
   n44 = Anim Dynamics - Bone: dyn_R_leg_fur_3(ComponentPose=@n43, Alpha=1, ExternalForce="0,0,0") {
    angularDampingOverride=0.9999
    angularSpringConstant=5000
    bAngularSpring=true
    bOverrideAngularDamping=true
    bUseGravityOverride=true
    gravityScale=0.35
    simulationSpace=BoneRelative
    alphaBoolBlend {blendOption=Linear}
    boundBone {boneName=dyn_R_leg_fur_3}
    componentLinearAccScale=<1 1 1>
    componentLinearVelScale=<1 1 1>
    gravityOverride {z=-980}
    physicsBodyDefinitions [{
      boundBone {boneName=dyn_R_leg_fur_3}
      boxExtents=<10 10 10>
      constraintSetup {angularLimitsMax=<0 30 0>, angularLimitsMin=<0 -30 0>, angularTarget=<1 0 0>}
      localJointOffset=<1 0 0>
     }]
    relativeSpaceBone {boneName=pelvis}
   }
   n45 = Anim Dynamics - Bone: dyn_R_leg_fur_4(ComponentPose=@n44, Alpha=1, ExternalForce="0,0,0") {
    angularDampingOverride=0.9999
    angularSpringConstant=5000
    bAngularSpring=true
    bOverrideAngularDamping=true
    bUseGravityOverride=true
    gravityScale=0.35
    simulationSpace=BoneRelative
    alphaBoolBlend {blendOption=Linear}
    boundBone {boneName=dyn_R_leg_fur_4}
    componentLinearAccScale=<1 1 1>
    componentLinearVelScale=<1 1 1>
    gravityOverride {z=-980}
    physicsBodyDefinitions [{
      boundBone {boneName=dyn_R_leg_fur_4}
      boxExtents=<10 10 10>
      constraintSetup {angularLimitsMax=<0 30 30>, angularLimitsMin=<0 -30 -30>, angularTarget=<1 0 0>}
      localJointOffset=<1 0 0>
     }]
    relativeSpaceBone {boneName=pelvis}
   }
   n46 = Anim Dynamics - Bone: dyn_R_leg_fur_5(ComponentPose=@n45, Alpha=1, ExternalForce="0,0,0") {
    angularDampingOverride=0.9999
    angularSpringConstant=5000
    bAngularSpring=true
    bOverrideAngularDamping=true
    bUseGravityOverride=true
    gravityScale=0.35
    simulationSpace=BoneRelative
    alphaBoolBlend {blendOption=Linear}
    boundBone {boneName=dyn_R_leg_fur_5}
    componentLinearAccScale=<1 1 1>
    componentLinearVelScale=<1 1 1>
    gravityOverride {z=-980}
    physicsBodyDefinitions [{
      boundBone {boneName=dyn_R_leg_fur_5}
      boxExtents=<10 10 10>
      constraintSetup {angularLimitsMax=<0 30 30>, angularLimitsMin=<0 -30 -30>, angularTarget=<1 0 0>}
      localJointOffset=<1 0 0>
     }]
    relativeSpaceBone {boneName=pelvis}
   }
   n47 = Anim Dynamics - Bone: dyn_R_leg_fur_6(ComponentPose=@n46, Alpha=1, ExternalForce="0,0,0") {
    angularDampingOverride=0.9999
    angularSpringConstant=5000
    bAngularSpring=true
    bOverrideAngularDamping=true
    bUseGravityOverride=true
    gravityScale=0.35
    simulationSpace=BoneRelative
    alphaBoolBlend {blendOption=Linear}
    boundBone {boneName=dyn_R_leg_fur_6}
    componentLinearAccScale=<1 1 1>
    componentLinearVelScale=<1 1 1>
    gravityOverride {z=-980}
    physicsBodyDefinitions [{
      boundBone {boneName=dyn_R_leg_fur_6}
      boxExtents=<10 10 10>
      constraintSetup {angularLimitsMax=<0 30 30>, angularLimitsMin=<0 -30 -30>, angularTarget=<1 0 0>}
      localJointOffset=<1 0 0>
     }]
    relativeSpaceBone {boneName=pelvis}
   }
   n48 = Anim Dynamics - Bone: dyn_R_leg_fur_7(ComponentPose=@n47, Alpha=1, ExternalForce="0,0,0") {
    angularDampingOverride=0.9999
    angularSpringConstant=5000
    bAngularSpring=true
    bOverrideAngularDamping=true
    bUseGravityOverride=true
    gravityScale=0.35
    simulationSpace=BoneRelative
    alphaBoolBlend {blendOption=Linear}
    boundBone {boneName=dyn_R_leg_fur_7}
    componentLinearAccScale=<1 1 1>
    componentLinearVelScale=<1 1 1>
    gravityOverride {z=-980}
    physicsBodyDefinitions [{
      boundBone {boneName=dyn_R_leg_fur_7}
      boxExtents=<10 10 10>
      constraintSetup {angularLimitsMax=<0 30 30>, angularLimitsMin=<0 -30 -30>, angularTarget=<1 0 0>}
      localJointOffset=<1 0 0>
     }]
    relativeSpaceBone {boneName=pelvis}
   }
   n49 = Anim Dynamics - Bone: dyn_R_leg_fur_8(ComponentPose=@n48, Alpha=1, ExternalForce="0,0,0") {
    angularDampingOverride=0.9999
    angularSpringConstant=5000
    bAngularSpring=true
    bOverrideAngularDamping=true
    bUseGravityOverride=true
    gravityScale=0.35
    simulationSpace=BoneRelative
    alphaBoolBlend {blendOption=Linear}
    boundBone {boneName=dyn_R_leg_fur_8}
    componentLinearAccScale=<1 1 1>
    componentLinearVelScale=<1 1 1>
    gravityOverride {z=-980}
    physicsBodyDefinitions [{
      boundBone {boneName=dyn_R_leg_fur_8}
      boxExtents=<10 10 10>
      constraintSetup {angularLimitsMax=<0 30 30>, angularLimitsMin=<0 -30 -30>, angularTarget=<1 0 0>}
      localJointOffset=<1 0 0>
     }]
    relativeSpaceBone {boneName=pelvis}
   }
   n50 = Anim Dynamics - Bone: dyn_L_foot_fur_1(ComponentPose=@n49, Alpha=1, ExternalForce="0,0,0") {
    angularDampingOverride=0.9999
    angularSpringConstant=5000
    bAngularSpring=true
    bOverrideAngularDamping=true
    bUseGravityOverride=true
    gravityScale=0.35
    simulationSpace=BoneRelative
    alphaBoolBlend {blendOption=Linear}
    boundBone {boneName=dyn_L_foot_fur_1}
    componentLinearAccScale=<1 1 1>
    componentLinearVelScale=<1 1 1>
    gravityOverride {z=-980}
    physicsBodyDefinitions [{
      boundBone {boneName=dyn_L_foot_fur_1}
      boxExtents=<10 10 10>
      constraintSetup {angularLimitsMax=<0 30 30>, angularLimitsMin=<0 -30 -30>, angularTarget=<1 0 0>}
      localJointOffset=<1 0 0>
     }]
    relativeSpaceBone {boneName=calf_l}
   }
   n51 = Anim Dynamics - Bone: dyn_L_foot_fur_2(ComponentPose=@n50, Alpha=1, ExternalForce="0,0,0") {
    angularDampingOverride=0.9999
    angularSpringConstant=5000
    bAngularSpring=true
    bOverrideAngularDamping=true
    bUseGravityOverride=true
    gravityScale=0.35
    simulationSpace=BoneRelative
    alphaBoolBlend {blendOption=Linear}
    boundBone {boneName=dyn_L_foot_fur_2}
    componentLinearAccScale=<1 1 1>
    componentLinearVelScale=<1 1 1>
    gravityOverride {z=-980}
    physicsBodyDefinitions [{
      boundBone {boneName=dyn_L_foot_fur_2}
      boxExtents=<10 10 10>
      constraintSetup {angularLimitsMax=<0 30 30>, angularLimitsMin=<0 -30 -30>, angularTarget=<1 0 0>}
      localJointOffset=<1 0 0>
     }]
    relativeSpaceBone {boneName=calf_l}
   }
   n52 = Anim Dynamics - Bone: dyn_L_foot_fur_3(ComponentPose=@n51, Alpha=1, ExternalForce="0,0,0") {
    angularDampingOverride=0.9999
    angularSpringConstant=5000
    bAngularSpring=true
    bOverrideAngularDamping=true
    bUseGravityOverride=true
    gravityScale=0.35
    simulationSpace=BoneRelative
    alphaBoolBlend {blendOption=Linear}
    boundBone {boneName=dyn_L_foot_fur_3}
    componentLinearAccScale=<1 1 1>
    componentLinearVelScale=<1 1 1>
    gravityOverride {z=-980}
    physicsBodyDefinitions [{
      boundBone {boneName=dyn_L_foot_fur_3}
      boxExtents=<10 10 10>
      constraintSetup {angularLimitsMax=<0 30 30>, angularLimitsMin=<0 -30 -30>, angularTarget=<1 0 0>}
      localJointOffset=<1 0 0>
     }]
    relativeSpaceBone {boneName=calf_l}
   }
   n53 = Anim Dynamics - Bone: dyn_L_foot_fur_4(ComponentPose=@n52, Alpha=1, ExternalForce="0,0,0") {
    angularDampingOverride=0.9999
    angularSpringConstant=5000
    bAngularSpring=true
    bOverrideAngularDamping=true
    bUseGravityOverride=true
    gravityScale=0.35
    simulationSpace=BoneRelative
    alphaBoolBlend {blendOption=Linear}
    boundBone {boneName=dyn_L_foot_fur_4}
    componentLinearAccScale=<1 1 1>
    componentLinearVelScale=<1 1 1>
    gravityOverride {z=-980}
    physicsBodyDefinitions [{
      boundBone {boneName=dyn_L_foot_fur_4}
      boxExtents=<10 10 10>
      constraintSetup {angularLimitsMax=<0 30 30>, angularLimitsMin=<0 -30 -30>, angularTarget=<1 0 0>}
      localJointOffset=<1 0 0>
     }]
    relativeSpaceBone {boneName=calf_l}
   }
   n54 = Anim Dynamics - Bone: dyn_L_foot_fur_5(ComponentPose=@n53, Alpha=1, ExternalForce="0,0,0") {
    angularDampingOverride=0.9999
    angularSpringConstant=5000
    bAngularSpring=true
    bOverrideAngularDamping=true
    bUseGravityOverride=true
    gravityScale=0.35
    simulationSpace=BoneRelative
    alphaBoolBlend {blendOption=Linear}
    boundBone {boneName=dyn_L_foot_fur_5}
    componentLinearAccScale=<1 1 1>
    componentLinearVelScale=<1 1 1>
    gravityOverride {z=-980}
    physicsBodyDefinitions [{
      boundBone {boneName=dyn_L_foot_fur_5}
      boxExtents=<10 10 10>
      constraintSetup {angularLimitsMax=<0 30 30>, angularLimitsMin=<0 -30 -30>, angularTarget=<1 0 0>}
      localJointOffset=<1 0 0>
     }]
    relativeSpaceBone {boneName=calf_l}
   }
   n55 = Anim Dynamics - Bone: dyn_L_foot_fur_6(ComponentPose=@n54, Alpha=1, ExternalForce="0,0,0") {
    angularDampingOverride=0.9999
    angularSpringConstant=5000
    bAngularSpring=true
    bOverrideAngularDamping=true
    bUseGravityOverride=true
    gravityScale=0.35
    simulationSpace=BoneRelative
    alphaBoolBlend {blendOption=Linear}
    boundBone {boneName=dyn_L_foot_fur_6}
    componentLinearAccScale=<1 1 1>
    componentLinearVelScale=<1 1 1>
    gravityOverride {z=-980}
    physicsBodyDefinitions [{
      boundBone {boneName=dyn_L_foot_fur_6}
      boxExtents=<10 10 10>
      constraintSetup {angularLimitsMax=<0 30 30>, angularLimitsMin=<0 -30 -30>, angularTarget=<1 0 0>}
      localJointOffset=<1 0 0>
     }]
    relativeSpaceBone {boneName=calf_l}
   }
   n56 = Anim Dynamics - Bone: dyn_R_foot_fur_1(ComponentPose=@n55, Alpha=1, ExternalForce="0,0,0") {
    angularDampingOverride=0.9999
    angularSpringConstant=5000
    bAngularSpring=true
    bOverrideAngularDamping=true
    bUseGravityOverride=true
    gravityScale=0.35
    simulationSpace=BoneRelative
    alphaBoolBlend {blendOption=Linear}
    boundBone {boneName=dyn_R_foot_fur_1}
    componentLinearAccScale=<1 1 1>
    componentLinearVelScale=<1 1 1>
    gravityOverride {z=-980}
    physicsBodyDefinitions [{
      boundBone {boneName=dyn_R_foot_fur_1}
      boxExtents=<10 10 10>
      constraintSetup {angularLimitsMax=<0 30 30>, angularLimitsMin=<0 -30 -30>, angularTarget=<1 0 0>}
      localJointOffset=<1 0 0>
     }]
    relativeSpaceBone {boneName=calf_r}
   }
   n57 = Anim Dynamics - Bone: dyn_R_foot_fur_2(ComponentPose=@n56, Alpha=1, ExternalForce="0,0,0") {
    angularDampingOverride=0.9999
    angularSpringConstant=5000
    bAngularSpring=true
    bOverrideAngularDamping=true
    bUseGravityOverride=true
    gravityScale=0.35
    simulationSpace=BoneRelative
    alphaBoolBlend {blendOption=Linear}
    boundBone {boneName=dyn_R_foot_fur_2}
    componentLinearAccScale=<1 1 1>
    componentLinearVelScale=<1 1 1>
    gravityOverride {z=-980}
    physicsBodyDefinitions [{
      boundBone {boneName=dyn_R_foot_fur_2}
      boxExtents=<10 10 10>
      constraintSetup {angularLimitsMax=<0 30 30>, angularLimitsMin=<0 -30 -30>, angularTarget=<1 0 0>}
      localJointOffset=<1 0 0>
     }]
    relativeSpaceBone {boneName=calf_r}
   }
   n58 = Anim Dynamics - Bone: dyn_R_foot_fur_3(ComponentPose=@n57, Alpha=1, ExternalForce="0,0,0") {
    angularDampingOverride=0.9999
    angularSpringConstant=5000
    bAngularSpring=true
    bOverrideAngularDamping=true
    bUseGravityOverride=true
    gravityScale=0.35
    simulationSpace=BoneRelative
    alphaBoolBlend {blendOption=Linear}
    boundBone {boneName=dyn_R_foot_fur_3}
    componentLinearAccScale=<1 1 1>
    componentLinearVelScale=<1 1 1>
    gravityOverride {z=-980}
    physicsBodyDefinitions [{
      boundBone {boneName=dyn_R_foot_fur_3}
      boxExtents=<10 10 10>
      constraintSetup {angularLimitsMax=<0 30 30>, angularLimitsMin=<0 -30 -30>, angularTarget=<1 0 0>}
      localJointOffset=<1 0 0>
     }]
    relativeSpaceBone {boneName=calf_r}
   }
   n59 = Anim Dynamics - Bone: dyn_R_foot_fur_4(ComponentPose=@n58, Alpha=1, ExternalForce="0,0,0") {
    angularDampingOverride=0.9999
    angularSpringConstant=5000
    bAngularSpring=true
    bOverrideAngularDamping=true
    bUseGravityOverride=true
    gravityScale=0.35
    simulationSpace=BoneRelative
    alphaBoolBlend {blendOption=Linear}
    boundBone {boneName=dyn_R_foot_fur_4}
    componentLinearAccScale=<1 1 1>
    componentLinearVelScale=<1 1 1>
    gravityOverride {z=-980}
    physicsBodyDefinitions [{
      boundBone {boneName=dyn_R_foot_fur_4}
      boxExtents=<10 10 10>
      constraintSetup {angularLimitsMax=<0 30 30>, angularLimitsMin=<0 -30 -30>, angularTarget=<1 0 0>}
      localJointOffset=<1 0 0>
     }]
    relativeSpaceBone {boneName=calf_r}
   }
   n60 = Anim Dynamics - Bone: dyn_R_foot_fur_5(ComponentPose=@n59, Alpha=1, ExternalForce="0,0,0") {
    angularDampingOverride=0.9999
    angularSpringConstant=5000
    bAngularSpring=true
    bOverrideAngularDamping=true
    bUseGravityOverride=true
    gravityScale=0.35
    simulationSpace=BoneRelative
    alphaBoolBlend {blendOption=Linear}
    boundBone {boneName=dyn_R_foot_fur_5}
    componentLinearAccScale=<1 1 1>
    componentLinearVelScale=<1 1 1>
    gravityOverride {z=-980}
    physicsBodyDefinitions [{
      boundBone {boneName=dyn_R_foot_fur_5}
      boxExtents=<10 10 10>
      constraintSetup {angularLimitsMax=<0 30 30>, angularLimitsMin=<0 -30 -30>, angularTarget=<1 0 0>}
      localJointOffset=<1 0 0>
     }]
    relativeSpaceBone {boneName=calf_r}
   }
   n61 = Anim Dynamics - Bone: dyn_R_foot_fur_6(ComponentPose=@n60, Alpha=1, ExternalForce="0,0,0") {
    angularDampingOverride=0.9999
    angularSpringConstant=5000
    bAngularSpring=true
    bOverrideAngularDamping=true
    bUseGravityOverride=true
    gravityScale=0.35
    simulationSpace=BoneRelative
    alphaBoolBlend {blendOption=Linear}
    boundBone {boneName=dyn_R_foot_fur_6}
    componentLinearAccScale=<1 1 1>
    componentLinearVelScale=<1 1 1>
    gravityOverride {z=-980}
    physicsBodyDefinitions [{
      boundBone {boneName=dyn_R_foot_fur_6}
      boxExtents=<10 10 10>
      constraintSetup {angularLimitsMax=<0 30 30>, angularLimitsMin=<0 -30 -30>, angularTarget=<1 0 0>}
      localJointOffset=<1 0 0>
     }]
    relativeSpaceBone {boneName=calf_r}
   }
   n62 = Anim Dynamics - Bone: dyn_L_legFlap!(ComponentPose=@n61, bAlphaBoolEnabled:bool=True, ExternalForce="0,0,0") -> (Pose) {
    alphaInputType=Bool
    angularDampingOverride=0.9999
    angularSpringConstant=5000
    bAngularSpring=true
    bOverrideAngularDamping=true
    bUseGravityOverride=true
    gravityScale=0.35
    linearDampingOverride=1
    linearSpringConstant=50
    simulationSpace=RootRelative
    alphaBoolBlend {blendOption=Linear}
    boundBone {boneName=dyn_L_legFlap}
    componentLinearAccScale=<1 1 1>
    componentLinearVelScale=<1 1 1>
    gravityOverride {z=-980}
    physicsBodyDefinitions [{
      boundBone {boneName=dyn_L_legFlap}
      boxExtents=<20 20 20>
      constraintSetup {angularLimitsMax=<0 5 30>, angularLimitsMin=<0 -5 -30>, angularTarget=<1 0 0>}
      localJointOffset=<1 0 0>
     }]
    relativeSpaceBone {boneName=thigh_l}
   }
   n63 = Anim Dynamics - Bone: dyn_R_legFlap!(ComponentPose=@n62, bAlphaBoolEnabled:bool=True, ExternalForce="0,0,0") -> (Pose) {
    alphaInputType=Bool
    angularDampingOverride=0.9999
    angularSpringConstant=5000
    bAngularSpring=true
    bOverrideAngularDamping=true
    bUseGravityOverride=true
    gravityScale=0.35
    linearDampingOverride=1
    linearSpringConstant=50
    simulationSpace=RootRelative
    alphaBoolBlend {blendOption=Linear}
    boundBone {boneName=dyn_R_legFlap}
    componentLinearAccScale=<1 1 1>
    componentLinearVelScale=<1 1 1>
    gravityOverride {z=-980}
    physicsBodyDefinitions [{
      boundBone {boneName=dyn_R_legFlap}
      boxExtents=<20 20 20>
      constraintSetup {angularLimitsMax=<0 5 30>, angularLimitsMin=<0 -5 -30>, angularTarget=<1 0 0>}
      localJointOffset=<1 0 0>
     }]
    relativeSpaceBone {boneName=thigh_r}
   }
   n64 = Get isMoving() -> (isMoving:bool)
   n65 = Anim Dynamics - Bone: dyn_L_calfTwist_WPO!(ComponentPose=@n63, bAlphaBoolEnabled:bool=@n64, ExternalForce="0,0,0") -> (Pose) {
    alphaInputType=Bool
    angularDampingOverride=1
    angularSpringConstant=20000
    bLinearSpring=true
    bOverrideLinearDamping=true
    bUseGravityOverride=true
    gravityScale=0.35
    linearDampingOverride=1
    linearSpringConstant=10
    simulationSpace=RootRelative
    alphaBoolBlend {blendInTime=0.4, blendOption=Linear, blendOutTime=0.4}
    boundBone {boneName=dyn_L_calfTwist_WPO}
    componentLinearAccScale=<1 1 1>
    componentLinearVelScale=<1 1 1>
    gravityOverride {z=-3000}
    physicsBodyDefinitions [{
      boundBone {boneName=dyn_L_calfTwist_WPO}
      boxExtents=<30 30 30>
      constraintSetup {angularLimitsMax=<0 0 0>, angularLimitsMin=<0 0 0>, angularTarget=<1 0 0>, linearAxesMax=<2 1 1>, linearAxesMin=<-2 -1 -1>}
      localJointOffset=<0 0 0>
     }]
    relativeSpaceBone {boneName=calf_l}
   }
   n66 = Anim Dynamics - Bone: dyn_R_calfTwist_WPO!(ComponentPose=@n65, bAlphaBoolEnabled:bool=@n64, ExternalForce="0,0,0") -> (Pose) {
    alphaInputType=Bool
    angularDampingOverride=1
    angularSpringConstant=20000
    bLinearSpring=true
    bOverrideLinearDamping=true
    bUseGravityOverride=true
    gravityScale=0.35
    linearDampingOverride=1
    linearSpringConstant=10
    simulationSpace=RootRelative
    alphaBoolBlend {blendInTime=0.4, blendOption=Linear, blendOutTime=0.4}
    boundBone {boneName=dyn_R_calfTwist_WPO}
    componentLinearAccScale=<1 1 1>
    componentLinearVelScale=<1 1 1>
    gravityOverride {z=-3000}
    physicsBodyDefinitions [{
      boundBone {boneName=dyn_R_calfTwist_WPO}
      boxExtents=<30 30 30>
      constraintSetup {angularLimitsMax=<0 0 0>, angularLimitsMin=<0 0 0>, angularTarget=<1 0 0>, linearAxesMax=<2 1 1>, linearAxesMin=<-2 -1 -1>}
      localJointOffset=<0 0 0>
     }]
    relativeSpaceBone {boneName=calf_r}
   }
   n67 = Get RigidBody Disabled() -> (RigidBody Disabled:bool)
   n68 = "Transform (Modify) Bone - Bone: dyn_overTie_1"!(ComponentPose=@n66, bAlphaBoolEnabled:bool=@n67, Translation="0,0,0", Rotation="-12,0,0", Scale="1,1,1") -> (Pose) {
    alphaInputType=Bool
    rotationMode=BMM_Additive
    rotationSpace=BCS_BoneSpace
    alphaBoolBlend {blendOption=Linear}
    boneToModify {boneName=dyn_overTie_1}
    rotation {pitch=-12}
   }
   n69 = "Transform (Modify) Bone - Bone: dyn_underTie_1"!(ComponentPose=@n68, bAlphaBoolEnabled:bool=@n67, Translation="0,0,0", Rotation="-12,0,0", Scale="1,1,1") -> (Pose) {
    alphaInputType=Bool
    rotationMode=BMM_Additive
    rotationSpace=BCS_BoneSpace
    alphaBoolBlend {blendOption=Linear}
    boneToModify {boneName=dyn_underTie_1}
    rotation {pitch=-12}
   }
   n70 = "Transform (Modify) Bone - Bone: dyn_belt"!(ComponentPose=@n69, bAlphaBoolEnabled:bool=Get bIsCrouching(), Scale="1,1,1") -> (Pose) {
    alphaInputType=Bool
    rotationMode=BMM_Additive
    translationMode=BMM_Additive
    alphaBoolBlend {blendInTime=0.3, blendOption=Linear, blendOutTime=0.3}
    boneToModify {boneName=dyn_belt}
    rotation {yaw=-20}
    translation {y=-2.5}
   }
   n71 = Get L_wristRotated() -> (L_wristRotated:double)
   n72 = "Transform (Modify) Bone - Bone: dyn_L_bracelet_3"!(ComponentPose=@n70, Alpha=@n71, Rotation="0,0,0", Scale="1,1,1") -> (Pose) {
    translationMode=BMM_Additive
    translationSpace=BCS_BoneSpace
    alphaBoolBlend {blendOption=Linear}
    boneToModify {boneName=dyn_L_bracelet_3}
    translation {x=-2}
   }
   n73 = "Transform (Modify) Bone - Bone: dyn_L_bracelet_2"!(ComponentPose=@n72, Alpha=@n71, Rotation="0,0,0", Scale="1,1,1") -> (Pose) {
    translationMode=BMM_Additive
    translationSpace=BCS_BoneSpace
    alphaBoolBlend {blendOption=Linear}
    boneToModify {boneName=dyn_L_bracelet_2}
    translation {x=-2}
   }
   n74 = "Transform (Modify) Bone - Bone: dyn_L_bracelet_1"!(ComponentPose=@n73, Alpha=@n71, Rotation="0,0,0", Scale="1,1,1") -> (Pose) {
    translationMode=BMM_Additive
    translationSpace=BCS_BoneSpace
    alphaBoolBlend {blendOption=Linear}
    boneToModify {boneName=dyn_L_bracelet_1}
    translation {x=-2}
   }
   n75 = Get R_wristRotated() -> (R_wristRotated:double)
   n76 = "Transform (Modify) Bone - Bone: dyn_R_bracelet_3"!(ComponentPose=@n74, Alpha=@n75, Rotation="0,0,0", Scale="1,1,1") -> (Pose) {
    translationMode=BMM_Additive
    translationSpace=BCS_BoneSpace
    alphaBoolBlend {blendOption=Linear}
    boneToModify {boneName=dyn_R_bracelet_3}
    translation {x=2}
   }
   n77 = "Transform (Modify) Bone - Bone: dyn_R_bracelet_2"!(ComponentPose=@n76, Alpha=@n75, Rotation="0,0,0", Scale="1,1,1") -> (Pose) {
    translationMode=BMM_Additive
    translationSpace=BCS_BoneSpace
    alphaBoolBlend {blendOption=Linear}
    boneToModify {boneName=dyn_R_bracelet_2}
    translation {x=2}
   }
   n78 = "Transform (Modify) Bone - Bone: dyn_R_bracelet_1"!(ComponentPose=@n77, Alpha=@n75, Rotation="0,0,0", Scale="1,1,1") -> (Pose) {
    translationMode=BMM_Additive
    translationSpace=BCS_BoneSpace
    alphaBoolBlend {blendOption=Linear}
    boneToModify {boneName=dyn_R_bracelet_1}
    translation {x=2}
   }
   Output Pose(Result=Component To Local(ComponentPose=@n78)) {layerGroup=None}
  ]}
 EventGraph {nodes [
   n1 = Event BlueprintUpdateAnimation()
   > n2 = Sequence!(execute) -> (then_0, then_1, then_2:Exec)
   > Is Valid(InputObject=Get playerPawn() -> (playerPawn:Fort Player Pawn Athena Object Reference))
   > Is Valid(InputObject=Get playerPawn())
   > Set WPO Globals Function(
    Material=Get materialInstanceDynamic() -> (materialInstanceDynamic:Material Instance Dynamic Object Reference)
    Skeletal Mesh=Get SkelMesh()
    "Local Noise Transform  Location"=Get Local Noise Transform Location() -> (Local Noise Transform Location:Vector)
    Local Noise Offset=Get Local Noise Offset()
   )
   > Set WPO Regional Function!(
    execute
    Material:Material Instance Dynamic Object Reference=Get materialInstanceDynamic()
    Overlay Material:Material Instance Dynamic Object Reference
    Skeletal Mesh:Skeletal Mesh Component Object Reference=Get SkelMesh()
    Velocity Transform:Name=root
    Velocity Interp Speed:double=30
    Collision Transform:Name=None
    Collision Transform Axis:EAxis Enum
    Invert Collision Transform Axis:bool=false
    Delta Time X:double=@n1.DeltaTimeX
    Previous Velocity Transform Location:"Vector (by ref)"=Get R1 Velocity Transform Location()
    Previous Velocity Transform Velocity:"Vector (by ref)"=Get R1 Velocity Bone Velocity() -> (R1 Velocity Bone Velocity:Vector)
    Region:int=1
   ) -> (then)
   > n3 = Branch(Condition=Get bIsSkyDiving())
   > n4 = Set WPO Pho Wind Intensity(WPO Pho Wind Intensity=1)
   SetScalarParameterValue(execute=[@n5.then, @n4.then], self=Get materialInstanceDynamic(), ParameterName=ClothSim_WindStrength, Value=Get WPO Pho Wind Intensity())
   n5 = Set WPO Pho Wind Intensity(execute=@n3.else, WPO Pho Wind Intensity=0)
   Set RigidBody Disabled(execute=@n2.then_1, RigidBody Disabled:bool=NOT Boolean(A=Get bRigidBodyEnabled())) -> (Output_Get:bool)
   > Set isMoving(isMoving:bool=OR Boolean!(A=Get bIsAccelerating2D(), B=Get bIsJumping(), C:bool=Get bIsFalling()) -> (ReturnValue)) -> (Output_Get:bool)
   n6 = GetOwnerSkeletalMeshComponent()
   n7 = BreakTransform(InTransform=GetDeltaTransformFromRefPose(self=@n6, BoneName=hand_l, BaseName=lowerarm_l))
   Set L_wristRotated(execute=@n2.then_2, L_wristRotated:double=MapRangeClamped(Value=BreakRotator(InRot=@n7.Rotation).Yaw, InRangeA=0, InRangeB=30)) -> (Output_Get:double)
   n8 = BreakTransform(InTransform=GetDeltaTransformFromRefPose(self=@n6, BoneName=hand_r, BaseName=lowerarm_r))
   > Set R_wristRotated(R_wristRotated:double=MapRangeClamped(Value=BreakRotator(InRot=@n8.Rotation).Yaw, InRangeA=0, InRangeB=30)) -> (Output_Get:double)
   Event BlueprintInitializeAnimation()
   > n9 = Sequence!(execute) -> (then_0, then_1)
   n10 = GetOwnerSkeletalMeshComponent()
   > Set SkelMesh(SkelMesh=@n10)
   n11 = TryGetPawnOwner()
   > Is Valid(InputObject=@n11)
   > n12 = Cast To FortPlayerPawnAthena(Object=@n11)
   > Set playerPawn(playerPawn:Fort Player Pawn Athena Object Reference=@n12) -> (Output_Get:Fort Player Pawn Athena Object Reference)
   n13 = Set Skel_Mesh(execute=@n9.then_1, Skel_Mesh=@n10)
   > Set Anim_Inst(Anim_Inst:Anim Instance Object Reference=GetAnimInstance(self=@n13)) -> (Output_Get:Anim Instance Object Reference)
   Event OnCharacterCustomizationLoaded()
   > n14 = Set SkelMesh(SkelMesh=GetPartSkeletalMeshComponent())
   > n15 = Get Dynamic Material Instance(SkeletalMeshComponent=@n14)
   > Set materialInstanceDynamic(materialInstanceDynamic:Material Instance Dynamic Object Reference=@n15) -> (Output_Get:Material Instance Dynamic Object Reference)
  ]}
]
variables [
 Anim_Inst {type=AnimInstance, default=null}
 isMoving {type=boolean}
 L_wristRotated {type=number}
 Local Noise Offset {type=Vector}
 Local Noise Transform Location {type=Vector}
 materialInstanceDynamic {type=MaterialInstanceDynamic}
 playerPawn {type=FortPlayerPawnAthena}
 R1 Velocity Bone Velocity {type=Vector}
 R1 Velocity Transform Location {type=Vector}
 R_wristRotated {type=number}
 RigidBody Disabled {type=boolean}
 Skel_Mesh {type=SkeletalMeshComponent}
 SkelMesh {type=SkeletalMeshComponent}
 WPO Pho Wind Intensity {type=number}
]
